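# Post-process Control Rig for the F_MED_MoralKore_Line female player body that corrects its dynamic skirt, strap, amulet and mask-holder bones.
# It smooths state bools into weights, then spherical pose readers on both thighs and the right calf drive skirt rotations and amulet corrections.
# During tactical sprint, a Clamp Capsule function constrains back-skirt bones against thigh capsules.
# FE_Idel and mask blend amulet and strap offsets, while the mask-holder offset always applies.
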ControlRigBlueprint /BRCosmetics/Characters/Player/Female/Medium/Bodies/F_MED_MoralKore_Line/Meshes/F_MED_MoralKore_Line_CtrlRig_post
graphs [
 New Function_ContainedGraph {nodes [
   RigVMFunctionEntryNode()
   > RigVMFunctionReturnNode()
  ]}
 New Function_ContainedGraph {nodes [
   n1 = RigVMFunctionEntryNode()
   Multiply = &RigVMFunction_MathTransformMul(A=@n1.CapsuleOffset, B=&RigUnit_GetTransform(bInitial=False, Item=@n1.CapsuleBone).Transform) {template="Multiply::Execute(in A,in B,out Result)"}
   RigUnit_GetTransform = &RigUnit_GetTransform(bInitial=False, Item=@n1.MoveBone)
   ClampSpatially = /Script/RigVM.RigVMFunction_MathTransformClampSpatially(
    DebugColor=(R=1,G=0,B=0,A=1)
    DebugThickness=1
    Type=Capsule
    Axis=@n1.CapsuleAxis
    Maximum=@n1.CapsuleLength
    Minimum=@n1.CapsuleWidth
    bDrawDebug=@n1.DrawDegub
    Space=@Multiply.Result
    Value=&RigVMFunction_MathTransformMul(B=@n1.BoneOffset, A=@RigUnit_GetTransform.Transform) {template="Multiply::Execute(in A,in B,out Result)"}.Result
   ) {
    is=struct RigVMUnitNode
    template="ClampSpatially::Execute(in Value,in Axis,in Type,in Minimum,in Maximum,in Space,in bDrawDebug,in DebugColor,in DebugThickness,out Result)"
   }
   Inverse = /Script/RigVM.RigVMFunction_MathTransformInverse(Value=@n1.BoneOffset) {is=struct RigVMUnitNode, template="Inverse::Execute(in Value,out Result)"}
   > RigUnit_SetTransform = &RigUnit_SetTransform(
    bInitial=False
    Weight=@n1.Alpha
    Value=&RigVMFunction_MathTransformMul(A=@ClampSpatially.Result, B=@Inverse.Result) {template="Multiply::Execute(in A,in B,out Result)"}.Result
    Item=@n1.MoveBone
   )
   > RigVMFunctionReturnNode()
   &RigVMFunction_MathTransformMul(B=(Rotation=(X=0,Y=0,Z=0,W=1),Translation=(X=-1,Y=-1,Z=-1),Scale3D=(X=1,Y=1,Z=1)), A=@n1.BoneOffset) {template="Multiply::Execute(in A,in B,out Result)"}
  ]}
 RigVMFunctionLibrary {nodes [
   Clamp Capsule = RigVMCollapseNode(
    Alpha=20
    BoneOffset=(Rotation=(X=0,Y=0,Z=0,W=1),Translation=(X=0,Y=0,Z=0),Scale3D=(X=1,Y=1,Z=1))
    CapsuleAxis=Z
    CapsuleBone=(Type=None,Name="None")
    CapsuleLength=20
    CapsuleOffset=(Rotation=(X=0,Y=0,Z=0,W=1),Translation=(X=0,Y=0,Z=0),Scale3D=(X=1,Y=1,Z=1))
    CapsuleWidth=20
    ExecuteContext=()
    MoveBone=(Type=None,Name="None")
   )
   New Function = RigVMCollapseNode(ExecuteContext=())
  ]}
 RigVMFunction_Sequence_1_ContainedGraph {nodes [
   n1 = RigVMFunctionEntryNode()
   n2 = &RigVMFunction_Sequence(ExecuteContext=@n1.ExecuteContext)
   n3 = &RigVMFunction_Sequence(ExecuteContext=@n2.B)
   RigVMFunctionReturnNode(B=@n3.A, C=@n3.B, A=@n2.A)
  ]}
 RigVMFunction_Sequence_1_ContainedGraph {nodes [
   n1 = RigVMFunctionEntryNode()
   n2 = &RigVMFunction_Sequence(ExecuteContext=@n1.ExecuteContext)
   n3 = &RigVMFunction_Sequence(ExecuteContext=@n2.B)
   RigVMFunctionReturnNode(B=@n3.A, C=@n3.B, A=@n2.A)
  ]}
 RigVMModel {nodes [
   RigUnit_BeginExecution = &RigUnit_BeginExecution()
   RigVMFunction_Sequence_3 = RigVMAggregateNode(ExecuteContext=@RigUnit_BeginExecution.ExecutePin)
   n1 = &RigVMFunction_Sequence(ExecuteContext=@RigVMFunction_Sequence_3.C)
   n2 = RigVMVariableNode() {variable=mask}
   RigVMFunction_MathBoolNot = &RigVMFunction_MathBoolNot(Value=@n2.Value)
   n3 = RigVMVariableNode() {variable=FE_Idel}
   RigVMFunction_MathBoolAnd = &RigVMFunction_MathBoolAnd(B=@RigVMFunction_MathBoolNot.Result, A=@n3.Value)
   RigVMFunction_MathBoolToFloat = &RigVMFunction_MathBoolToFloat(Value=@RigVMFunction_MathBoolAnd.Result)
   n4 = &RigVMFunction_AlphaInterp(InterpSpeedDecreasing=10, InterpSpeedIncreasing=10, Scale=1, bClampResult=False, Value=@RigVMFunction_MathBoolToFloat.Result)
   &RigUnit_OffsetTransformForItem(
    Item=(Type=Bone,Name="dyn_amulet")
    OffsetTransform=(Rotation=(X=-0.0296,Y=0.1643,Z=-0.0049,W=-0.986),Translation=(X=-0.0608,Y=0.1377,Z=0.3421),Scale3D=(X=1,Y=1,Z=1))
    bPropagateToChildren=false
    ExecutePin=@n1.A
    Weight=@n4.Result
   )
   > &RigUnit_OffsetTransformForItem(
    Item=(Type=Bone,Name="dyn_strap_skirt_05")
    OffsetTransform=(Rotation=(X=0,Y=0,Z=-0.2846,W=-0.9587),Translation=(X=0.6204,Y=0,Z=-1.645),Scale3D=(X=1,Y=1,Z=1))
    bPropagateToChildren=false
    Weight=@n4.Result
   )
   n5 = &RigVMFunction_AlphaInterp(
    InterpSpeedDecreasing=10
    InterpSpeedIncreasing=10
    Scale=1
    bClampResult=False
    Value=&RigVMFunction_MathBoolToFloat(Value=&RigVMFunction_MathBoolAnd(A=@n3.Value, B=@n2.Value).Result).Result
   )
   &RigUnit_OffsetTransformForItem(
    Item=(Type=Bone,Name="dyn_amulet")
    OffsetTransform=(Rotation=(X=-0.0296,Y=0.1643,Z=-0.0049,W=-0.986),Translation=(X=-0.0608,Y=0.1377,Z=0.3421),Scale3D=(X=1,Y=1,Z=1))
    bPropagateToChildren=false
    ExecutePin=@n1.B
    Weight=@n5.Result
   )
   > &RigUnit_OffsetTransformForItem(
    Item=(Type=Bone,Name="dyn_strap_skirt_05")
    OffsetTransform=(Rotation=(X=0,Y=0,Z=-0.2846,W=-0.9587),Translation=(X=0.6204,Y=0,Z=-1.645),Scale3D=(X=1,Y=1,Z=1))
    bPropagateToChildren=false
    Weight=@n5.Result
   )
   > &RigUnit_OffsetTransformForItem(
    Item=(Type=Bone,Name="dyn_maskholder")
    OffsetTransform=(Rotation=(X=0,Y=0.2079,Z=0,W=0.9781),Translation=(X=0.3,Y=1.1,Z=0.2),Scale3D=(X=1,Y=1,Z=1))
    Weight=1
    bPropagateToChildren=false
   )
   RigVMFunction_Sequence_4 = RigVMAggregateNode(ExecuteContext=@RigVMFunction_Sequence_3.A)
   n6 = &RigVMFunction_AlphaInterp(
    InterpSpeedDecreasing=10
    InterpSpeedIncreasing=10
    Scale=1
    bClampResult=true
    bMapRange=true
    Value=&RigVMFunction_MathBoolToFloat(Value=RigVMVariableNode() {variable=move}.Value).Result
   )
   RigVMVariableNode(Value=@n6.Result, ExecuteContext=@RigVMFunction_Sequence_4.C) {variable=moveAlpha}
   n7 = &RigVMFunction_AlphaInterp(
    InterpSpeedDecreasing=10
    InterpSpeedIncreasing=10
    OutRange=(Min=0,Max=0.2)
    Scale=1
    bClampResult=true
    bInterpResult=False
    bMapRange=true
    Value=&RigVMFunction_MathBoolToFloat(Value=RigVMVariableNode() {variable=parachute}.Value).Result
   )
   RigVMVariableNode(Value=@n7.Result, ExecuteContext=@RigVMFunction_Sequence_4.A) {variable=parachuteAlpha}
   n8 = &RigVMFunction_AlphaInterp(
    InterpSpeedDecreasing=10
    InterpSpeedIncreasing=10
    OutRange=(Min=1,Max=0)
    Scale=1
    bClampResult=true
    bMapRange=true
    Value=&RigVMFunction_MathBoolToFloat(Value=RigVMVariableNode() {variable=emote}.Value).Result
   )
   RigVMVariableNode(Value=@n8.Result, ExecuteContext=@RigVMFunction_Sequence_4.B) {variable=emoteAlpha}
   RigVMFunction_Sequence = &RigVMFunction_Sequence(ExecuteContext=@RigVMFunction_Sequence_3.B)
   n9 = &RigVMFunction_Sequence(ExecuteContext=@RigVMFunction_Sequence.B)
   n10 = &RigUnit_SphericalPoseReader(
    ActiveRegionScaleFactors=(PositiveWidth=1,NegativeWidth=2,PositiveHeight=1,NegativeHeight=1)
    ActiveRegionSize=0.1
    Debug=(bDrawDebug=false,DebugScale=25,DebugSegments=20,DebugThickness=0.25)
    DriverAxis=(X=-1,Y=0,Z=0)
    DriverItem=(Type=Bone,Name="thigh_l")
    FalloffRegionScaleFactors=(PositiveWidth=1,NegativeWidth=1,PositiveHeight=1,NegativeHeight=1)
    FalloffSize=0.3
    OptionalParentItem=(Type=Bone,Name="None")
    RotationOffset=(X=97,Y=9,Z=107)
    ExecutePin=@n9.A
   )
   n11 = RigVMVariableNode() {variable=moveAlpha}
   n12 = RigVMVariableNode() {variable=parachuteAlpha}
   > &RigUnit_ModifyTransforms(
    ItemToModify=((Item=(Type=Bone,Name="dyn_skirt_bk_centr_base"),Transform=(Rotation=(X=0,Y=-0.2164,Z=0,W=0.9763),Translation=(X=0,Y=0,Z=0),Scale3D=(X=1,Y=1,Z=1))))
    WeightMinimum=0
    Weight=&RigVMFunction_MathDoubleMul(B=&RigVMFunction_MathDoubleMax(B=@n11.Value, A=@n12.Value).Result, A=@n10.OutputParam).Result
   )
   > &RigUnit_ParentConstraint(
    AdvancedSettings=(InterpolationType=Average,RotationOrderForFilter=XZY)
    Child=(Type=Bone,Name="dyn_amulet")
    Parents=((Item=(Type=Bone,Name="dyn_strap_skirt_04"),Weight=1))
    Weight=@n10.OutputParam
   )
   > &RigUnit_OffsetTransformForItem(
    Item=(Type=Bone,Name="dyn_amulet")
    OffsetTransform=(Rotation=(X=0.018,Y=0.0742,Z=0.0755,W=0.9942),Translation=(X=0,Y=0,Z=0),Scale3D=(X=1,Y=1,Z=1))
    bPropagateToChildren=false
    Weight=@n10.OutputParam
   )
   n13 = &RigVMFunction_AlphaInterp(
    InterpSpeedDecreasing=10
    InterpSpeedIncreasing=10
    Scale=1
    bClampResult=False
    Value=&RigVMFunction_MathBoolToFloat(Value=RigVMVariableNode() {variable=TacticalSprinting}.Value).Result
   )
   > Clamp Capsule_5 = RigVMFunctionReferenceNode(
    BoneOffset=(Rotation=(X=0,Y=0,Z=0,W=1),Translation=(X=0,Y=0,Z=0),Scale3D=(X=1,Y=1,Z=1))
    CapsuleAxis=X
    CapsuleBone=(Type=Bone,Name="thigh_l")
    CapsuleLength=40
    CapsuleOffset=(Rotation=(X=0,Y=0,Z=0,W=1),Translation=(X=-15,Y=2,Z=0),Scale3D=(X=1,Y=1,Z=1))
    CapsuleWidth=10
    DrawDegub=false
    MoveBone=(Type=Bone,Name="dyn_skirt_bk_l_02")
    Alpha=@n13.Result
   )
   > Clamp Capsule_6 = RigVMFunctionReferenceNode(
    BoneOffset=(Rotation=(X=0,Y=0,Z=0,W=1),Translation=(X=0,Y=0,Z=0),Scale3D=(X=1,Y=1,Z=1))
    CapsuleAxis=X
    CapsuleBone=(Type=Bone,Name="thigh_l")
    CapsuleLength=40
    CapsuleOffset=(Rotation=(X=0,Y=0,Z=0,W=1),Translation=(X=-15,Y=2,Z=0),Scale3D=(X=1,Y=1,Z=1))
    CapsuleWidth=23
    DrawDegub=false
    MoveBone=(Type=Bone,Name="dyn_skirt_bk_centr_02")
    Alpha=@n13.Result
   )
   n14 = &RigUnit_SphericalPoseReader(
    ActiveRegionScaleFactors=(PositiveWidth=1,NegativeWidth=2,PositiveHeight=1,NegativeHeight=1)
    ActiveRegionSize=0.1
    Debug=(bDrawDebug=false,DebugScale=25,DebugSegments=20,DebugThickness=0.25)
    DriverAxis=(X=1,Y=0,Z=0)
    DriverItem=(Type=Bone,Name="thigh_r")
    FalloffRegionScaleFactors=(PositiveWidth=1,NegativeWidth=1,PositiveHeight=1,NegativeHeight=1)
    FalloffSize=0.3
    OptionalParentItem=(Type=Bone,Name="None")
    RotationOffset=(X=90,Y=0,Z=-65)
    ExecutePin=@n9.B
   )
   Maximum = &RigVMFunction_MathDoubleMax(A=RigVMVariableNode() {variable=parachuteAlpha}.Value, B=RigVMVariableNode() {variable=moveAlpha}.Value)
   n15 = &RigVMFunction_MathDoubleMul(B=@Maximum.Result, A=@n14.OutputParam)
   > &RigUnit_ModifyTransforms(
    ItemToModify=((Item=(Type=Bone,Name="dyn_skirt_bk_centr_base"),Transform=(Rotation=(X=0,Y=-0.2164,Z=0,W=0.9763),Translation=(X=0,Y=0,Z=0),Scale3D=(X=1,Y=1,Z=1))))
    WeightMinimum=0
    Weight=@n15.Result
   )
   n16 = RigVMVariableNode() {variable=jump}
   > &RigUnit_ModifyTransforms(
    ItemToModify=((Item=(Type=Bone,Name="dyn_skirt_bk_r_01"),Transform=(Rotation=(X=0,Y=-0.1305,Z=0,W=0.9914),Translation=(X=0,Y=0,Z=0),Scale3D=(X=1,Y=1,Z=1))))
    WeightMinimum=0
    Weight=&RigVMFunction_MathDoubleMul(B=&RigVMFunction_MathBoolToFloat(Value=@n16.Value).Result, A=@n15.Result).Result
   )
   n17 = &RigVMFunction_AlphaInterp(
    InterpSpeedDecreasing=10
    InterpSpeedIncreasing=10
    Scale=1
    bClampResult=False
    Value=&RigVMFunction_MathBoolToFloat(Value=RigVMVariableNode() {variable=TacticalSprinting}.Value).Result
   )
   > Clamp Capsule_1 = RigVMFunctionReferenceNode(
    BoneOffset=(Rotation=(X=0,Y=0,Z=0,W=1),Translation=(X=0,Y=0,Z=0),Scale3D=(X=1,Y=1,Z=1))
    CapsuleAxis=X
    CapsuleBone=(Type=Bone,Name="thigh_r")
    CapsuleLength=40
    CapsuleOffset=(Rotation=(X=0,Y=0,Z=0,W=1),Translation=(X=15,Y=2,Z=0),Scale3D=(X=1,Y=1,Z=1))
    CapsuleWidth=10
    DrawDegub=false
    MoveBone=(Type=Bone,Name="dyn_skirt_bk_r_02")
    Alpha=@n17.Result
   )
   > Clamp Capsule_2 = RigVMFunctionReferenceNode(
    BoneOffset=(Rotation=(X=0,Y=0,Z=0,W=1),Translation=(X=0,Y=0,Z=0),Scale3D=(X=1,Y=1,Z=1))
    CapsuleAxis=X
    CapsuleBone=(Type=Bone,Name="thigh_r")
    CapsuleLength=40
    CapsuleOffset=(Rotation=(X=0,Y=0,Z=0,W=1),Translation=(X=15,Y=2,Z=0),Scale3D=(X=1,Y=1,Z=1))
    CapsuleWidth=10
    DrawDegub=false
    MoveBone=(Type=Bone,Name="dyn_skirt_bk_r_03")
    Alpha=@n17.Result
   )
   n18 = &RigUnit_SphericalPoseReader(
    ActiveRegionScaleFactors=(PositiveWidth=3,NegativeWidth=3,PositiveHeight=3,NegativeHeight=3)
    ActiveRegionSize=0.01
    Debug=(bDrawDebug=True,bDraw2D=false,DebugScale=25,DebugSegments=20,DebugThickness=0.25)
    DriverAxis=(X=1,Y=0,Z=0)
    DriverItem=(Type=Bone,Name="calf_r")
    FalloffRegionScaleFactors=(PositiveWidth=1,NegativeWidth=1,PositiveHeight=1,NegativeHeight=1)
    FalloffSize=0.1
    FlipHeightScaling=false
    FlipWidthScaling=true
    OptionalParentItem=(Type=Bone)
    RotationOffset=(X=90,Y=-3,Z=47)
    ExecutePin=@RigVMFunction_Sequence.A
   )
   AlphaInterp = &RigVMFunction_AlphaInterp(InterpSpeedDecreasing=6, InterpSpeedIncreasing=6, Scale=1, bClampResult=False, Value=@n18.OutputParam)
   > &RigUnit_ModifyTransforms(
    ItemToModify=((Item=(Type=Bone,Name="dyn_skirt_bk_r_02"),Transform=(Rotation=(X=0,Y=-0.4695,Z=0,W=0.8829),Translation=(X=0,Y=0,Z=0),Scale3D=(X=0.5,Y=1,Z=1))))
    WeightMinimum=0
    Weight=@AlphaInterp.Result
   )
   Multiply = &RigVMFunction_MathDoubleMul(A=&RigVMFunction_MathBoolToFloat(Value=RigVMVariableNode() {variable=DBNO}.Value).Result, B=@n10.OutputParam)
   &RigUnit_OffsetTransformForItem(
    ExecutePin=()
    Item=(Type=Bone,Name="dyn_maskholder")
    OffsetTransform=(Rotation=(X=0,Y=0.2079,Z=0,W=0.9781),Translation=(X=0.3,Y=1.2,Z=0.2),Scale3D=(X=1,Y=1,Z=1))
    Weight=1
    bPropagateToChildren=false
   )
   RigVMRerouteNode(Value=@n14.OutputParam)
  ]}
]
hierarchy [
 root {loc=<0 0 0>, rot=<0 0 0 1>}
 pelvis {parent=root, loc=<0 -0.5301 98.0571>, rot=<0.0127 0.707 -0.0127 -0.707>}
 spine_01 {parent=pelvis, loc=<6.1208 -0.9328 0>, rot=<0 0 0.034 -0.9994>}
 spine_02 {parent=spine_01, loc=<6.5997 -0.2095 0>, rot=<0 0 0.0238 -0.9997>}
 spine_03 {parent=spine_02, loc=<6.4914 -0.1237 0>, rot=<0 0 0.0149 -0.9999>}
 spine_04 {parent=spine_03, loc=<6.9598 0.0794 0>, rot=<0 0 -0.0132 -0.9999>}
 spine_05 {parent=spine_04, loc=<6.9987 0 0>, rot=<0 0 -0.0015 -1>}
 neck_01 {parent=spine_05, loc=<11.3138 0.1529 0>, rot=<0 0 -0.2003 -0.9797>}
 neck_02 {parent=neck_01, loc=<4.5627 0 0>, rot=<0 0 0.0504 -0.9987>}
 head {parent=neck_02, loc=<5.3914 0 0>, rot=<0 0 0.115 -0.9934>}
 FACIAL_C_FacialRoot {parent=head, loc=<2.1524 -0.746 -0.0002>, rot=<0.4879 -0.5118 -0.4879 -0.5118>}
 FACIAL_C_UnderChin {parent=FACIAL_C_FacialRoot, loc=<-0.0005 8.5709 3.8336>, rot=<0.4988 0 0 -0.8667>}
 FACIAL_L_UnderChin {parent=FACIAL_C_FacialRoot, loc=<2.7503 7.3986 2.9191>, rot=<0.4015 -0.355 0.0819 -0.8403>}
 FACIAL_R_UnderChin {parent=FACIAL_C_FacialRoot, loc=<-2.7509 7.3979 2.9217>, rot=<0.4015 0.3549 -0.0819 -0.8403>}
 FACIAL_C_Jaw {parent=FACIAL_C_FacialRoot, loc=<-0.0014 2.4518 2.0283>, rot=<0.3409 0 0 -0.9401>}
 FACIAL_C_Jawline {parent=FACIAL_C_Jaw, loc=<0.0007 2.2502 7.0022>, rot=<0.2586 0 0 -0.966>}
 FACIAL_L_Jawline {parent=FACIAL_C_FacialRoot, loc=<3.9004 5.5494 4.2859>, rot=<0.2082 0.1578 -0.0544 -0.9637>}
 FACIAL_R_Jawline {parent=FACIAL_C_FacialRoot, loc=<-3.9018 5.5502 4.2874>, rot=<0.2082 -0.1578 0.0545 -0.9637>}
 clavicle_l {parent=spine_05, loc=<5.4831 -0.1053 -3.2946>, rot=<0.0465 -0.7289 -0.009 -0.6829>}
 upperarm_l {parent=clavicle_l, loc=<9.8658 -1.2136 -0.4105>, rot=<0.0164 -0.3245 -0.0269 -0.9454>}
 lowerarm_l {parent=upperarm_l, loc=<26.465 -0.8345 0.259>, rot=<0.0036 0.0161 -0.2298 -0.9731>}
 hand_l {parent=lowerarm_l, loc=<22.7355 -0.3203 -0.6105>, rot=<-0.6784 -0.0223 0.0735 -0.7307>}
 thumb_01_l {parent=hand_l, loc=<1.3118 -0.3191 -2.459>, rot=<0.6546 -0.4042 0.0664 -0.6354>}
 thumb_02_l {parent=thumb_01_l, loc=<5.9336 -0.0448 0.0401>, rot=<-0.0037 0.0213 0.298 -0.9543>}
 thumb_03_l {parent=thumb_02_l, loc=<2.2223 0.3682 -0.0907>, rot=<0.0118 -0.013 0.0066 -0.9998>}
 index_metacarpal_l {parent=hand_l, loc=<1.6535 0.0183 -2.96>, rot=<0.0158 -0.0124 -0.0173 -0.9996>}
 index_01_l {parent=index_metacarpal_l, loc=<8.3008 0.0357 -0.0524>, rot=<-0.0044 0.0485 0.0924 -0.9945>}
 index_02_l {parent=index_01_l, loc=<3.3587 -0.3004 0.0452>, rot=<-0.002 0.0027 0.3225 -0.9466>}
 index_03_l {parent=index_02_l, loc=<2.5578 -0.123 -0.0252>, rot=<-0.002 0.0067 0.2558 -0.9667>}
 middle_metacarpal_l {parent=hand_l, loc=<2.1856 -0.1725 -0.5984>, rot=<0.006 0.0168 -0.0329 -0.9993>}
 middle_01_l {parent=middle_metacarpal_l, loc=<7.3275 -0.2447 0.0129>, rot=<-0.0208 0.0042 0.0858 -0.9961>}
 middle_02_l {parent=middle_01_l, loc=<4.0123 0.1401 0.1175>, rot=<-0.0218 0.0258 0.3512 -0.9357>}
 middle_03_l {parent=middle_02_l, loc=<2.6077 0.0534 0.1373>, rot=<0.0082 0.0188 0.1708 -0.9851>}
 ring_metacarpal_l {parent=hand_l, loc=<2.2074 -0.3625 1.3638>, rot=<-0.0284 0.0722 -0.0458 -0.9959>}
 ring_01_l {parent=ring_metacarpal_l, loc=<7.0295 -0.6197 -0.1446>, rot=<-0.0489 -0.0252 0.1892 -0.9804>}
 ring_02_l {parent=ring_01_l, loc=<3.6174 0.4647 -0.1027>, rot=<0 0 0.2806 -0.9598>}
 ring_03_l {parent=ring_02_l, loc=<2.4152 0.114 0.0199>, rot=<0 0 0.1189 -0.9929>}
 pinky_metacarpal_l {parent=hand_l, loc=<1.7616 -0.9753 3.3814>, rot=<-0.0397 0.0814 -0.069 -0.9935>}
 pinky_01_l {parent=pinky_metacarpal_l, loc=<7.1709 -0.4831 -0.1094>, rot=<-0.0654 -0.0035 0.2242 -0.9723>}
 pinky_02_l {parent=pinky_01_l, loc=<2.8274 0.0893 -0.1043>, rot=<0.0037 0.0011 0.2854 -0.9584>}
 pinky_03_l {parent=pinky_02_l, loc=<2.2046 0.0307 -0.0078>, rot=<-0.0002 0.0098 0.1445 -0.9895>}
 deform_wrist_dn_base_l {parent=hand_l, loc=<0 -1.5326 0>, rot=<0 0 0 1>}
 deform_wrist_dn_l {parent=deform_wrist_dn_base_l, loc=<0 0 0>, rot=<0 0 0 1>}
 deform_wrist_up_base_l {parent=hand_l, loc=<0 2.3009 0>, rot=<0 0 0 1>}
 deform_wrist_up_l {parent=deform_wrist_up_base_l, loc=<0 0 0>, rot=<0 0 0 1>}
 lowerarm_twist_02_l {parent=lowerarm_l, loc=<10.6707 0 0>, rot=<0 0 0 1>}
 lowerarm_twist_01_l {parent=lowerarm_l, loc=<18.316 0 0>, rot=<0 0 0 1>}
 deform_elbow_in_base_l {parent=lowerarm_l, loc=<0.1834 3.6062 -0.2572>, rot=<0 0 0 1>}
 deform_elbow_in_l {parent=deform_elbow_in_base_l, loc=<0 0 0>, rot=<0 0 0 1>}
 deform_elbow_out_base_l {parent=lowerarm_l, loc=<0 -4 0>, rot=<0 0 0 1>}
 deform_elbow_out_l {parent=deform_elbow_out_base_l, loc=<0 0 0>, rot=<0 0 0 1>}
 upperarm_twist_01_l {parent=upperarm_l, loc=<7.2135 -0.1134 0>, rot=<0 0 0 1>}
 deform_upperarm_bk_base_l {parent=upperarm_twist_01_l, loc=<-6.4053 -4.3746 0.314>, rot=<0 0 0 1>}
 deform_upperarm_bk_l {parent=deform_upperarm_bk_base_l, loc=<0 0 0>, rot=<0 0 0 1>}
 deform_upperarm_fr_base_l {parent=upperarm_twist_01_l, loc=<-5.6312 5.092 0.3354>, rot=<0 0 0 1>}
 deform_upperarm_fr_l {parent=deform_upperarm_fr_base_l, loc=<0 0 0>, rot=<0 0 0 1>}
 deform_upperarm_up_base_l {parent=upperarm_twist_01_l, loc=<-8.891 0.4825 3.9517>, rot=<0 0 0 1>}
 deform_upperarm_up_l {parent=deform_upperarm_up_base_l, loc=<0 0 0>, rot=<0 0 0 1>}
 upperarm_twist_02_l {parent=upperarm_l, loc=<19.7998 -0.1134 0>, rot=<0 0 0 1>}
 clavicle_r {parent=spine_05, loc=<5.4831 -0.1053 3.2946>, rot=<0.7289 0.0465 -0.6829 0.009>}
 upperarm_r {parent=clavicle_r, loc=<-9.8658 1.2136 0.4105>, rot=<0.0164 -0.3245 -0.0269 -0.9454>}
 lowerarm_r {parent=upperarm_r, loc=<-26.465 0.8345 -0.259>, rot=<0.0036 0.0161 -0.2298 -0.9731>}
 hand_r {parent=lowerarm_r, loc=<-22.7355 0.3203 0.6106>, rot=<-0.6784 -0.0223 0.0735 -0.7307>}
 thumb_01_r {parent=hand_r, loc=<-1.3118 0.3191 2.459>, rot=<0.6546 -0.4042 0.0664 -0.6354>}
 thumb_02_r {parent=thumb_01_r, loc=<-5.9336 0.0448 -0.0401>, rot=<-0.0037 0.0213 0.298 -0.9543>}
 thumb_03_r {parent=thumb_02_r, loc=<-2.2223 -0.3682 0.0907>, rot=<0.0118 -0.013 0.0066 -0.9998>}
 index_metacarpal_r {parent=hand_r, loc=<-1.6535 -0.0183 2.96>, rot=<0.0158 -0.0124 -0.0173 -0.9996>}
 index_01_r {parent=index_metacarpal_r, loc=<-8.3008 -0.0357 0.0524>, rot=<-0.0044 0.0485 0.0924 -0.9945>}
 index_02_r {parent=index_01_r, loc=<-3.3587 0.3004 -0.0452>, rot=<-0.002 0.0027 0.3225 -0.9466>}
 index_03_r {parent=index_02_r, loc=<-2.5578 0.123 0.0252>, rot=<-0.002 0.0067 0.2558 -0.9667>}
 middle_metacarpal_r {parent=hand_r, loc=<-2.1856 0.1725 0.5984>, rot=<0.006 0.0168 -0.0329 -0.9993>}
 middle_01_r {parent=middle_metacarpal_r, loc=<-7.3275 0.2447 -0.0129>, rot=<-0.0208 0.0042 0.0858 -0.9961>}
 middle_02_r {parent=middle_01_r, loc=<-4.0123 -0.1401 -0.1175>, rot=<-0.0218 0.0258 0.3512 -0.9357>}
 middle_03_r {parent=middle_02_r, loc=<-2.6077 -0.0534 -0.1373>, rot=<0.0082 0.0188 0.1708 -0.9851>}
 ring_metacarpal_r {parent=hand_r, loc=<-2.2074 0.3625 -1.3638>, rot=<-0.0284 0.0722 -0.0458 -0.9959>}
 ring_01_r {parent=ring_metacarpal_r, loc=<-7.0295 0.6197 0.1446>, rot=<-0.0489 -0.0252 0.1892 -0.9804>}
 ring_02_r {parent=ring_01_r, loc=<-3.6174 -0.4647 0.1027>, rot=<0 0 0.2806 -0.9598>}
 ring_03_r {parent=ring_02_r, loc=<-2.4152 -0.114 -0.0199>, rot=<0 0 0.1189 -0.9929>}
 pinky_metacarpal_r {parent=hand_r, loc=<-1.7616 0.9753 -3.3814>, rot=<-0.0397 0.0814 -0.069 -0.9935>}
 pinky_01_r {parent=pinky_metacarpal_r, loc=<-7.1709 0.4831 0.1094>, rot=<-0.0654 -0.0035 0.2242 -0.9723>}
 pinky_02_r {parent=pinky_01_r, loc=<-2.8274 -0.0893 0.1043>, rot=<0.0037 0.0011 0.2854 -0.9584>}
 pinky_03_r {parent=pinky_02_r, loc=<-2.2046 -0.0307 0.0078>, rot=<-0.0002 0.0098 0.1445 -0.9895>}
 deform_wrist_dn_base_r {parent=hand_r, loc=<-0.0001 1.5328 0>, rot=<0 0 0 1>}
 deform_wrist_dn_r {parent=deform_wrist_dn_base_r, loc=<0 0 0>, rot=<0 0 0 1>}
 deform_wrist_up_base_r {parent=hand_r, loc=<-0.0002 -2.3007 -0.0001>, rot=<0 0 0 1>}
 deform_wrist_up_r {parent=deform_wrist_up_base_r, loc=<0 0 0>, rot=<0 0 0 1>}
 lowerarm_twist_02_r {parent=lowerarm_r, loc=<-10.6707 0 0>, rot=<0 0 0 -1>}
 lowerarm_twist_01_r {parent=lowerarm_r, loc=<-18.316 0 0>, rot=<0 0 0 -1>}
 deform_elbow_in_base_r {parent=lowerarm_r, loc=<-0.1834 -3.6062 0.2572>, rot=<0 0 0 -1>}
 deform_elbow_in_r {parent=deform_elbow_in_base_r, loc=<0 0 0>, rot=<0 0 0 1>}
 deform_elbow_out_base_r {parent=lowerarm_r, loc=<0 4 -0.0001>, rot=<0 0 0 -1>}
 deform_elbow_out_r {parent=deform_elbow_out_base_r, loc=<0 0 0>, rot=<0 0 0 1>}
 upperarm_twist_01_r {parent=upperarm_r, loc=<-7.2135 0.1134 0>, rot=<0 0 0 -1>}
 deform_upperarm_bk_base_r {parent=upperarm_twist_01_r, loc=<6.4053 4.3746 -0.314>, rot=<0 0 0 1>}
 deform_upperarm_bk_r {parent=deform_upperarm_bk_base_r, loc=<0 0 0>, rot=<0 0 0 1>}
 deform_upperarm_fr_base_r {parent=upperarm_twist_01_r, loc=<5.6312 -5.092 -0.3354>, rot=<0 0 0 1>}
 deform_upperarm_fr_r {parent=deform_upperarm_fr_base_r, loc=<0 0 0>, rot=<0 0 0 1>}
 deform_upperarm_up_base_r {parent=upperarm_twist_01_r, loc=<8.891 -0.4825 -3.9517>, rot=<0 0 0 1>}
 deform_upperarm_up_r {parent=deform_upperarm_up_base_r, loc=<0 0 0>, rot=<0 0 0 1>}
 upperarm_twist_02_r {parent=upperarm_r, loc=<-19.7998 0.1134 0>, rot=<0 0 0 -1>}
 attach_backpack {parent=spine_05, loc=<-0.1187 -13.4736 -0.0002>, rot=<-0.0093 -0.707 0.0093 -0.707>}
 Backpack_BR {parent=attach_backpack, type=Null, loc=<0 3.6789 0>, rot=<0 0 0 1>}
 attach_cape {parent=spine_05, loc=<9.0203 -11.7146 -0.0002>, rot=<0.0616 -0.7044 -0.0616 -0.7044>}
 cape {parent=attach_cape, type=Null, loc=<0 0 0>, rot=<0 0 0 1>}
 extra_armor_back {parent=spine_05, loc=<-4.3603 -7.5556 0>, rot=<0.0284 -0.7065 -0.0284 -0.7065>}
 deform_pec_aim_base_l {parent=spine_04, loc=<3.2456 9.3309 -1.2237>, rot=<0.0603 -0.4033 0.3318 -0.8507>}
 deform_pec_aim_l {parent=deform_pec_aim_base_l, loc=<0 0 0>, rot=<0 0 0 1>}
 deform_pec_base_l {parent=deform_pec_aim_l, loc=<9.3665 0 0>, rot=<0 0 0 1>}
 deform_pec_l {parent=deform_pec_base_l, loc=<0 0 0>, rot=<0 0 0 1>}
 deform_pec_aim_base_r {parent=spine_04, loc=<3.2452 9.3309 1.2237>, rot=<-0.4033 -0.0603 0.8507 0.3318>}
 deform_pec_aim_r {parent=deform_pec_aim_base_r, loc=<0 0 0>, rot=<0 0 0 1>}
 deform_pec_base_r {parent=deform_pec_aim_r, loc=<-9.3667 -0.0003 -0.0004>, rot=<0 0 0 1>}
 deform_pec_r {parent=deform_pec_base_r, loc=<0 0 0>, rot=<0 0 0 1>}
 thigh_l {parent=pelvis, loc=<-1.0855 -0.2109 -8.7053>, rot=<0.0009 0.0355 0.0155 -0.9993>}
 calf_l {parent=thigh_l, loc=<-42.5284 -0.0019 -0.7297>, rot=<0.0006 -0.0161 -0.0391 -0.9991>}
 foot_l {parent=calf_l, loc=<-46.7401 -2.5014 -0.592>, rot=<0.0427 -0.0238 0.0353 -0.9982>}
 ball_l {parent=foot_l, loc=<-3.5658 10.8066 -0.1544>, rot=<0.0001 -0.0001 -0.7063 -0.7079>}
 calf_twist_02_l {parent=calf_l, loc=<-19.1546 -1.7943 -0.2638>, rot=<0.0237 0.0095 -0.0035 -0.9997>}
 calf_twist_01_l {parent=calf_l, loc=<-35.8483 -1.9197 -0.5794>, rot=<0.0237 0.0095 -0.0035 -0.9997>}
 deform_knee_in_base_l {parent=calf_l, loc=<0.8031 -5.0231 0.2456>, rot=<0 0 0 -1>}
 deform_knee_in_l {parent=deform_knee_in_base_l, loc=<0 0 0>, rot=<0 0 0 1>}
 deform_knee_out_base_l {parent=calf_l, loc=<0 4 0>, rot=<0 0 0 -1>}
 deform_knee_out_l {parent=deform_knee_out_base_l, loc=<0 0 0>, rot=<0 0 0 1>}
 extra_knee_pad {parent=calf_l, loc=<-2.2879 4.0878 0.0831>, rot=<0.4049 0.5167 0.5691 0.4951>}
 thigh_twist_01_l {parent=thigh_l, loc=<-27.952 0 0>, rot=<0.0013 0 0 -1>}
 deform_groin_base_l {parent=thigh_l, loc=<0.4801 5.6776 -0.2486>, rot=<0 0 0 -1>}
 deform_groin_l {parent=deform_groin_base_l, loc=<0 0 0>, rot=<0 0 0 1>}
 dyn_thigh_01_l {parent=thigh_l, loc=<0.5252 0 0>, rot=<0 0 0 -1>}
 dyn_skirt_fr2_01_l_base {parent=dyn_thigh_01_l, loc=<2.6454 5.3242 -3.6121>, rot=<-0.0336 -0.931 -0.3555 -0.0753>}
 dyn_skirt_fr2_01_l {parent=dyn_skirt_fr2_01_l_base, loc=<0.1998 -0.0001 0.0001>, rot=<0 0 0 1>}
 dyn_skirt_fr2_02_l {parent=dyn_skirt_fr2_01_l, loc=<10.7913 0.0001 -0.0001>, rot=<0 -0.0807 0 -0.9967>}
 dyn_skirt_fr_01_l_base {parent=dyn_thigh_01_l, loc=<2.447 7.1546 1.2688>, rot=<-0.0368 -0.7412 -0.669 -0.0428>}
 dyn_skirt_fr_01_l {parent=dyn_skirt_fr_01_l_base, loc=<0.1998 0 0>, rot=<0 0 0 1>}
 dyn_skirt_fr_02_l {parent=dyn_skirt_fr_01_l, loc=<10.7912 0 0>, rot=<0 -0.0641 0 -0.9979>}
 thigh_r {parent=pelvis, loc=<-1.0855 -0.2109 8.7053>, rot=<0.0355 -0.0009 0.9993 0.0155>}
 dyn_thigh_01_r {parent=thigh_r, loc=<-0.5252 0 0>, rot=<0 -1 0 0>}
 dyn_skirt_fr_01_r_base {parent=dyn_thigh_01_r, loc=<2.447 -7.1546 1.2688>, rot=<0.0368 -0.7412 0.669 -0.0428>}
 dyn_skirt_fr_01_r {parent=dyn_skirt_fr_01_r_base, loc=<0.1997 0 0>, rot=<0 0 0 1>}
 dyn_skirt_fr_02_r {parent=dyn_skirt_fr_01_r, loc=<10.7912 0 0>, rot=<0 -0.0641 0 -0.9979>}
 dyn_skirt_fr_03_r {parent=dyn_skirt_fr_02_r, loc=<10.7908 0 0>, rot=<0 0 0 1>}
 dyn_skirt_fr2_01_r_base {parent=dyn_thigh_01_r, loc=<2.645 -5.3242 -3.6121>, rot=<0.0336 -0.931 0.3555 -0.0753>}
 dyn_skirt_fr2_01_r {parent=dyn_skirt_fr2_01_r_base, loc=<0.1997 0.0001 0.0001>, rot=<0 0 0 1>}
 dyn_skirt_fr2_02_r {parent=dyn_skirt_fr2_01_r, loc=<10.7911 0 0>, rot=<0 -0.0807 0 -0.9967>}
 dyn_skirt_fr2_03_r {parent=dyn_skirt_fr2_02_r, loc=<10.7908 0 0>, rot=<0 0 0 1>}
 calf_r {parent=thigh_r, loc=<42.5284 0.0019 0.7297>, rot=<0.0006 -0.0161 -0.0391 -0.9991>}
 foot_r {parent=calf_r, loc=<46.7401 2.5014 0.592>, rot=<0.0427 -0.0238 0.0353 -0.9982>}
 ball_r {parent=foot_r, loc=<3.5658 -10.8066 0.1544>, rot=<0.0001 -0.0001 -0.7063 -0.7079>}
 calf_twist_02_r {parent=calf_r, loc=<19.1546 1.7943 0.2638>, rot=<0.0237 0.0095 -0.0035 -0.9997>}
 calf_twist_01_r {parent=calf_r, loc=<35.8483 1.9197 0.5794>, rot=<0.0237 0.0095 -0.0035 -0.9997>}
 deform_knee_in_base_r {parent=calf_r, loc=<-0.8031 5.0231 -0.2457>, rot=<0 0 0 -1>}
 deform_knee_in_r {parent=deform_knee_in_base_r, loc=<0 0 0>, rot=<0 0 0 1>}
 deform_knee_out_base_r {parent=calf_r, loc=<0 -4 0>, rot=<0 0 0 -1>}
 deform_knee_out_r {parent=deform_knee_out_base_r, loc=<0 0 0>, rot=<0 0 0 1>}
 thigh_twist_01_r {parent=thigh_r, loc=<27.952 0 0>, rot=<0.0013 0 0 -1>}
 deform_groin_base_r {parent=thigh_r, loc=<-0.4801 -5.6776 0.2486>, rot=<0 0 0 -1>}
 deform_groin_r {parent=deform_groin_base_r, loc=<0 0 0>, rot=<0 0 0 1>}
 deform_glute_base_l {parent=pelvis, loc=<-7.3466 -9.0213 -9.1245>, rot=<0.0172 0.031 -0.4477 -0.8935>}
 deform_glute_l {parent=deform_glute_base_l, loc=<0 0 0>, rot=<0 0 0 1>}
 deform_glute_base_r {parent=pelvis, loc=<-7.3464 -9.0213 9.1245>, rot=<0.031 -0.0172 0.8935 -0.4477>}
 deform_glute_r {parent=deform_glute_base_r, loc=<0 0 0>, rot=<0 0 0 1>}
 dyn_skirt_bk_l_base {parent=pelvis, loc=<-3.9783 -9.8822 -10.926>, rot=<-0.0661 -0.8687 0.4904 0.0219>}
 dyn_skirt_bk_l_01 {parent=dyn_skirt_bk_l_base, loc=<0.5 0 0>, rot=<0 0 0 -1>}
 dyn_skirt_bk_l_02 {parent=dyn_skirt_bk_l_01, loc=<9 0.0001 0>, rot=<0 0 0 -1>}
 dyn_skirt_bk_centr_base {parent=pelvis, loc=<-4.2963 -12.1245 0>, rot=<-0.0627 -0.7037 0.7053 0.0585>}
 dyn_skirt_bk_centr_01 {parent=dyn_skirt_bk_centr_base, loc=<0.6855 0 0>, rot=<0 0 0 1>}
 dyn_skirt_bk_centr_02 {parent=dyn_skirt_bk_centr_01, loc=<9 0 0.4974>, rot=<0 -0.0381 0 -0.9993>}
 dyn_skirt_side_l_base {parent=pelvis, loc=<1.1705 -1.4758 -13.8486>, rot=<-0.0079 -0.9902 0.027 -0.1365>}
 dyn_skirt_side_l_01 {parent=dyn_skirt_side_l_base, loc=<0.6851 0 0.0003>, rot=<0 0 0 1>}
 dyn_skirt_side_l_02 {parent=dyn_skirt_side_l_01, loc=<8.8709 0 -0.0002>, rot=<0 -0.083 0 -0.9965>}
 dyn_skirt_bk_r_base {parent=pelvis, loc=<-3.9783 -9.8822 10.926>, rot=<-0.0219 -0.4904 0.8687 0.0661>}
 dyn_skirt_bk_r_01 {parent=dyn_skirt_bk_r_base, loc=<0.5 0 0>, rot=<0 0 0 1>}
 dyn_skirt_bk_r_02 {parent=dyn_skirt_bk_r_01, loc=<9 0 0>, rot=<0 0 0 1>}
 dyn_skirt_bk_r_03 {parent=dyn_skirt_bk_r_02, loc=<9 0 0>, rot=<0 0 0 1>}
 dyn_skirt_fr_01_base {parent=pelvis, loc=<2.2991 7.6035 0>, rot=<0.001 -0.7112 -0.703 -0.0082>}
 dyn_skirt_fr_01 {parent=dyn_skirt_fr_01_base, loc=<0.2 0 0>, rot=<0 0 0 1>}
 dyn_skirt_fr_02 {parent=dyn_skirt_fr_01, loc=<10.7908 0 0>, rot=<0 0 0 1>}
 dyn_skirt_side_r_base {parent=pelvis, loc=<-2.1227 -1.3578 15.2169>, rot=<0.0891 -0.0266 0.9956 0.0092>}
 dyn_skirt_side_r_01 {parent=dyn_skirt_side_r_base, loc=<0.6855 0 0>, rot=<0 0 0 1>}
 dyn_skirt_side_r_02 {parent=dyn_skirt_side_r_01, loc=<8.8706 0 0>, rot=<0 -0.0723 0 -0.9974>}
 dyn_skirt_side_r_03 {parent=dyn_skirt_side_r_02, loc=<8.8706 0 0>, rot=<0 -0.0533 0 -0.9986>}
 dyn_strap_skirt_fr_base {parent=pelvis, loc=<7.2574 5.7934 -7.8187>, rot=<-0.3989 0.194 0.8912 0.0943>}
 dyn_strap_skirt_01 {parent=dyn_strap_skirt_fr_base, loc=<0.4911 0 0>, rot=<0 0 0 1>}
 dyn_strap_skirt_02 {parent=dyn_strap_skirt_01, loc=<2.9132 0 0>, rot=<-0.0991 -0.1922 -0.1679 -0.9618>}
 dyn_strap_skirt_03 {parent=dyn_strap_skirt_02, loc=<3.747 0 0>, rot=<-0.0575 -0.2479 -0.2477 -0.9348>}
 dyn_strap_skirt_04 {parent=dyn_strap_skirt_03, loc=<3.683 0 0>, rot=<-0.6836 -0.3301 0.0269 -0.6504>}
 dyn_strap_skirt_05 {parent=dyn_strap_skirt_04, loc=<5.3993 0 0>, rot=<0.0409 -0.1874 0.2624 -0.9457>}
 dyn_strap_skirt_06 {parent=dyn_strap_skirt_05, loc=<4.1495 0 0>, rot=<0.0339 -0.1429 0.2888 -0.946>}
 dyn_strap_skirt_07 {parent=dyn_strap_skirt_06, loc=<4.1676 0 0>, rot=<0.0237 0 0 -0.9997>}
 dyn_amulet {parent=dyn_strap_skirt_04, loc=<-0.5952 2.122 0.3594>, rot=<0.0678 0.0501 -0.7828 -0.6165>}
 dyn_maskholder {parent=dyn_amulet, loc=<0.6089 0.9195 0.2774>, rot=<-0.0328 -0.0818 0.1634 -0.9826>}
 dyn_strap_skirt_back_base {parent=pelvis, loc=<9.7179 -6.5682 -8.2118>, rot=<0.8816 -0.0944 0.3423 0.3111>}
 dyn_collision_strap {parent=pelvis, loc=<5.1199 -0.6215 -11.3337>, rot=<-0.0127 -0.707 0.0127 -0.707>}
]
variables [
 DBNO {type=bool, default=False}
 FE_Idel {type=bool, default=False}
 TacticalSprinting {type=bool, default=False}
 emote {type=bool, default=False}
 emoteAlpha {type=double, default=0}
 jump {type=bool, default=False}
 mask {type=bool, default=False}
 move {type=bool, default=False}
 moveAlpha {type=double, default=0}
 parachute {type=bool, default=False}
 parachuteAlpha {type=double, default=0}
]
