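# Physics asset for the MoralKore Line female medium body cosmetic, with capsule bodies for its skirt, strap, amulet and mask holder.
# Back and side skirt panels are tightly limited chains on free-rotating bases off spine_01, with right-side tips cross-linked.
# A strap chain spans front and back pelvis anchors, branching mid-chain to the amulet and mask holder.
# Pelvis, spine, thighs, right calf and a collision strap supply body shapes.
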PhysicsAsset /BRCosmetics/Characters/Player/Female/Medium/Bodies/F_MED_MoralKore_Line/Meshes/F_MED_MoralKore_Line_Physics_FE
bodies {
 calf_r [{shapeType=Capsule, length=40, radius=8, shapeName=calf_r_capsule, center=<26 0 0>, rotation=<rot 90 0 0>}]
 dyn_amulet [
  {shapeType=Capsule, length=10, radius=3, shapeName=dyn_amulet_capsule, center=<3 0 -0.8927>, rotation=<rot 0 0.0006 84.9998>}
  {shapeType=Capsule, length=10.06, radius=5, shapeName=dyn_amulet_capsule, center=<-4.7863 0 0.434>, rotation=<rot 0 0.0011 79.9997>}
 ]
 dyn_collision_strap [{
   shapeType=Capsule
   length=19.1469
   radius=5.2672
   shapeName=dyn_collision_strap_capsule
   center=<-2.5643 0 -5.1451>
   rotation=<rot -0.0004 -90 -15>
  }]
 dyn_maskholder [{shapeType=Capsule, length=9, radius=1, shapeName=dyn_maskholder_capsule, center=<7 0 -1.4209>, rotation=<rot 90 0 0>}]
 dyn_skirt_bk_centr_01 [{shapeType=Capsule, length=8, radius=2.5, center=<4.7075 0 1.5194>, rotation=<rot 0 0 90>}]
 dyn_skirt_bk_centr_02 [{shapeType=Capsule, length=10, radius=3, center=<5 0 0>, rotation=<rot 0 0 90>}]
 dyn_skirt_bk_centr_base [{shapeType=Capsule, length=1, radius=1}]
 dyn_skirt_bk_l_01 [{shapeType=Capsule, length=8, radius=2.5, center=<4 -2 0>, rotation=<rot 0 0 90>}]
 dyn_skirt_bk_l_02 [{shapeType=Capsule, length=10, radius=3, center=<5 0 0>, rotation=<rot 0 0 90>}]
 dyn_skirt_bk_l_base [{shapeType=Capsule, length=1, radius=1}]
 dyn_skirt_bk_r_01 [{shapeType=Capsule, length=8, radius=2.5, center=<4 2 0>, rotation=<rot 0 0 90>}]
 dyn_skirt_bk_r_02 [{shapeType=Capsule, length=10, radius=2, center=<5 0 0>, rotation=<rot 0 0 90>}]
 dyn_skirt_bk_r_03 [{shapeType=Capsule, length=10, radius=2, center=<4 0 0>, rotation=<rot 0 0 90>}]
 dyn_skirt_bk_r_base [{shapeType=Capsule, length=1, radius=1}]
 dyn_skirt_side_l_01 [{shapeType=Capsule, length=8, radius=2, center=<4 0 0>, rotation=<rot 0 0 90>}]
 dyn_skirt_side_l_02 [{shapeType=Capsule, length=10, radius=3, center=<5 0 0>, rotation=<rot 0 0 90>}]
 dyn_skirt_side_l_base [{shapeType=Capsule, length=1, radius=1}]
 dyn_skirt_side_r_01 [{shapeType=Capsule, length=8, radius=3, center=<4 0 0>, rotation=<rot 0 0 90>}]
 dyn_skirt_side_r_02 [{shapeType=Capsule, length=10, radius=3, center=<5 0 0>, rotation=<rot 0 0 90>}]
 dyn_skirt_side_r_03 [{shapeType=Capsule, length=10, radius=4, center=<4 0 0>, rotation=<rot 0 0 90>}]
 dyn_skirt_side_r_base [{shapeType=Capsule, length=1, radius=1}]
 dyn_strap_skirt_01 [{shapeType=Capsule, length=2, radius=1, shapeName=dyn_strap_skirt_01_capsule, center=<3 0 0>, rotation=<rot 90 0 0>}]
 dyn_strap_skirt_02 [{shapeType=Capsule, length=2, radius=1, shapeName=dyn_strap_skirt_02_capsule, center=<3 0 0>, rotation=<rot 90 0 0>}]
 dyn_strap_skirt_03 [{shapeType=Capsule, length=2, radius=2, shapeName=dyn_strap_skirt_03_capsule, center=<3 0 0>, rotation=<rot 90 0 0>}]
 dyn_strap_skirt_04 [{shapeType=Capsule, length=2, radius=3, shapeName=dyn_strap_skirt_04_capsule, center=<3 0 0>, rotation=<rot 90 0 0>}]
 dyn_strap_skirt_05 [{shapeType=Capsule, length=2, radius=0.5, shapeName=dyn_strap_skirt_05_capsule, center=<3 0 0>, rotation=<rot 90 0 0>}]
 dyn_strap_skirt_06 [{shapeType=Capsule, length=2, radius=1.5, shapeName=dyn_strap_skirt_06_capsule, center=<3 0 0>, rotation=<rot 90 0 0>}]
 dyn_strap_skirt_07 [{shapeType=Capsule, length=2, radius=2, shapeName=dyn_strap_skirt_07_capsule, rotation=<rot 90 0 0>}]
 dyn_strap_skirt_back_base [{shapeType=Capsule, length=2, radius=2, shapeName=dyn_strap_skirt_back_base_capsule, rotation=<rot 0 -90 0>}]
 dyn_strap_skirt_fr_base [{shapeType=Capsule, length=2, radius=1, shapeName=dyn_strap_skirt_fr_base_capsule}]
 pelvis [
  {shapeType=Capsule, length=6, radius=9, center=<-1.7446 2.3079 0>}
  {shapeType=Capsule, length=8, radius=9, shapeName=pelvis_capsule, center=<-0.6396 0.1353 0>}
 ]
 spine_01 [{shapeType=Capsule, length=6, radius=8, center=<0 1.4539 0>}]
 thigh_l [{shapeType=Capsule, length=20, radius=10, center=<-17.2057 -0.0085 1.4762>, rotation=<rot 0 -87.6286 -97.9391>}]
 thigh_r [{shapeType=Capsule, length=20, radius=8, center=<17.2057 0.0085 -1.4762>, rotation=<rot 0 -87.6286 82.0609>}]
}
constraints [
 {
  bone1Name=dyn_collision_strap
  bone2Name=pelvis
  swing1LimitDegrees=45
  swing1Motion=Free
  swing2LimitDegrees=45
  swing2Motion=Free
  twistLimitDegrees=45
  twistMotion=Free
 }
 {
  bone1Name=dyn_skirt_bk_centr_base
  bone2Name=spine_01
  swing1LimitDegrees=45
  swing1Motion=Free
  swing2LimitDegrees=45
  swing2Motion=Free
  twistLimitDegrees=45
  twistMotion=Free
 }
 {bone1Name=dyn_skirt_bk_centr_01, bone2Name=dyn_skirt_bk_centr_base, swing1LimitDegrees=10, swing2LimitDegrees=10, twistLimitDegrees=1}
 {bone1Name=dyn_skirt_bk_centr_02, bone2Name=dyn_skirt_bk_centr_01, swing1LimitDegrees=10, swing2LimitDegrees=10, twistLimitDegrees=1}
 {
  bone1Name=dyn_skirt_bk_l_base
  bone2Name=spine_01
  swing1LimitDegrees=45
  swing1Motion=Free
  swing2LimitDegrees=45
  swing2Motion=Free
  twistLimitDegrees=45
  twistMotion=Free
 }
 {bone1Name=dyn_skirt_bk_l_01, bone2Name=dyn_skirt_bk_l_base, swing1LimitDegrees=10, swing2LimitDegrees=10, twistLimitDegrees=1}
 {bone1Name=dyn_skirt_bk_l_02, bone2Name=dyn_skirt_bk_l_01, swing1LimitDegrees=10, swing2LimitDegrees=10, twistLimitDegrees=1}
 {
  bone1Name=dyn_skirt_bk_r_base
  bone2Name=spine_01
  swing1LimitDegrees=45
  swing1Motion=Free
  swing2LimitDegrees=45
  swing2Motion=Free
  twistLimitDegrees=45
  twistMotion=Free
 }
 {bone1Name=dyn_skirt_bk_r_01, bone2Name=dyn_skirt_bk_r_base, swing1LimitDegrees=10, swing2LimitDegrees=10, twistLimitDegrees=1}
 {bone1Name=dyn_skirt_bk_r_02, bone2Name=dyn_skirt_bk_r_01, swing1LimitDegrees=5, swing2LimitDegrees=5, twistLimitDegrees=1}
 {bone1Name=dyn_skirt_bk_r_03, bone2Name=dyn_skirt_bk_r_02, swing1LimitDegrees=5, swing2LimitDegrees=5, twistLimitDegrees=1}
 {
  bone1Name=dyn_skirt_side_r_03
  bone2Name=dyn_skirt_bk_r_03
  swing1LimitDegrees=45
  swing1Motion=Free
  swing2LimitDegrees=45
  swing2Motion=Free
  twistLimitDegrees=45
  twistMotion=Free
 }
 {
  bone1Name=dyn_skirt_side_l_base
  bone2Name=spine_01
  swing1LimitDegrees=45
  swing1Motion=Free
  swing2LimitDegrees=45
  swing2Motion=Free
  twistLimitDegrees=45
  twistMotion=Free
 }
 {bone1Name=dyn_skirt_side_l_01, bone2Name=dyn_skirt_side_l_base, swing1LimitDegrees=10, swing2LimitDegrees=10, twistLimitDegrees=5}
 {bone1Name=dyn_skirt_side_l_02, bone2Name=dyn_skirt_side_l_01, swing1LimitDegrees=5, swing2LimitDegrees=15, twistLimitDegrees=1}
 {
  bone1Name=dyn_skirt_side_r_base
  bone2Name=spine_01
  swing1LimitDegrees=45
  swing1Motion=Free
  swing2LimitDegrees=45
  swing2Motion=Free
  twistLimitDegrees=45
  twistMotion=Free
 }
 {bone1Name=dyn_skirt_side_r_01, bone2Name=dyn_skirt_side_r_base, swing1LimitDegrees=5, swing2LimitDegrees=10, twistLimitDegrees=5}
 {bone1Name=dyn_skirt_side_r_02, bone2Name=dyn_skirt_side_r_01, swing1LimitDegrees=5, swing2LimitDegrees=15, twistLimitDegrees=1}
 {bone1Name=dyn_skirt_side_r_03, bone2Name=dyn_skirt_side_r_02, swing1LimitDegrees=5, swing2LimitDegrees=15, twistLimitDegrees=1}
 {
  bone1Name=dyn_strap_skirt_fr_base
  bone2Name=pelvis
  swing1LimitDegrees=45
  swing1Motion=Free
  swing2LimitDegrees=45
  swing2Motion=Free
  twistLimitDegrees=45
  twistMotion=Free
 }
 {bone1Name=dyn_strap_skirt_01, bone2Name=dyn_strap_skirt_fr_base, swing1LimitDegrees=5, swing2LimitDegrees=5, twistLimitDegrees=5}
 {bone1Name=dyn_strap_skirt_02, bone2Name=dyn_strap_skirt_01, swing1LimitDegrees=5, swing2LimitDegrees=5, twistLimitDegrees=5}
 {bone1Name=dyn_strap_skirt_03, bone2Name=dyn_strap_skirt_02, swing1LimitDegrees=5, swing2LimitDegrees=5, twistLimitDegrees=5}
 {bone1Name=dyn_strap_skirt_04, bone2Name=dyn_strap_skirt_03, swing1LimitDegrees=5, swing2LimitDegrees=5, twistLimitDegrees=5}
 {bone1Name=dyn_amulet, bone2Name=dyn_strap_skirt_04, swing1LimitDegrees=30, swing2LimitDegrees=30, twistLimitDegrees=5}
 {bone1Name=dyn_maskholder, bone2Name=dyn_amulet, swing1LimitDegrees=60, swing2LimitDegrees=60, twistLimitDegrees=5}
 {bone1Name=dyn_strap_skirt_05, bone2Name=dyn_strap_skirt_04, swing1LimitDegrees=5, swing2LimitDegrees=5, twistLimitDegrees=5}
 {bone1Name=dyn_strap_skirt_06, bone2Name=dyn_strap_skirt_05, swing1LimitDegrees=5, swing2LimitDegrees=5, twistLimitDegrees=5}
 {bone1Name=dyn_strap_skirt_07, bone2Name=dyn_strap_skirt_06, swing1LimitDegrees=5, swing2LimitDegrees=5, twistLimitDegrees=5}
 {bone1Name=spine_01, bone2Name=pelvis, swing1LimitDegrees=45, swing2LimitDegrees=45, twistLimitDegrees=45}
 {bone1Name=thigh_l, bone2Name=pelvis, swing1LimitDegrees=45, swing2LimitDegrees=45, twistLimitDegrees=45}
 {bone1Name=thigh_r, bone2Name=pelvis, swing1LimitDegrees=45, swing2LimitDegrees=45, twistLimitDegrees=45}
 {
  bone1Name=calf_r
  bone2Name=thigh_r
  swing1LimitDegrees=45
  swing1Motion=Free
  swing2LimitDegrees=45
  swing2Motion=Free
  twistLimitDegrees=45
  twistMotion=Free
 }
 {
  bone1Name=dyn_strap_skirt_back_base
  bone2Name=pelvis
  swing1LimitDegrees=45
  swing1Motion=Free
  swing2LimitDegrees=45
  swing2Motion=Free
  twistLimitDegrees=45
  twistMotion=Free
 }
 {
  bone1Name=dyn_strap_skirt_back_base
  bone2Name=dyn_strap_skirt_07
  swing1LimitDegrees=45
  swing1Motion=Free
  swing2LimitDegrees=45
  twistLimitDegrees=45
 }
]
