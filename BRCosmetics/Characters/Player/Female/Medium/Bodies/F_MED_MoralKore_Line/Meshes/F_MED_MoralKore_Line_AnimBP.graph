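# Cosmetic AnimBP for the F_MED_MoralKore_Line body that copies the parent mesh pose and drives secondary motion on skirt, straps, amulet and armor.
# Bone constraints and a pre Control Rig feed two RigidBody simulations, using standard and front-end physics assets.
# Cached poses then blend crouch and mask-variant corrections before strap look-ats, a post Control Rig and final offsets.
# The event graph updates movement, front-end, mask-tag and gravity state.
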AnimBlueprint /BRCosmetics/Characters/Player/Female/Medium/Bodies/F_MED_MoralKore_Line/Meshes/F_MED_MoralKore_Line_AnimBP
target_skeleton=&/F_MED_MoralKore_Line_Skeleton
graphs [
 AnimGraph {nodes [
   Blend Poses by bool(BlendPose_0=Copy Pose From Mesh() {bUseAttachedParent=true})
   n1 = "Constraint - dyn_strap_skirt_04 (P)"(Alpha=0.9, ConstraintWeights_0=1) {
    alphaBoolBlend {blendOption=Linear}
    boneToModify {boneName=dyn_strap_skirt_04}
    constraintSetup [{targetBone {boneName=dyn_amulet}}]
    constraintWeights [1]
   }
   "Constraint - dyn_amulet (P)"(ComponentPose=@n1, Alpha=0.9, ConstraintWeights_0=1) {
    alphaBoolBlend {blendOption=Linear}
    boneToModify {boneName=dyn_amulet}
    constraintSetup [{targetBone {boneName=pelvis}}]
    constraintWeights [1]
   }
   n2 = "Constraint - dyn_strap_skirt_05 (P)"(ComponentPose=Local To Component(LocalPose=Use cached pose 'mask'()), Alpha=0.5, ConstraintWeights_0=1) {
    alphaBoolBlend {blendOption=Linear}
    boneToModify {boneName=dyn_strap_skirt_05}
    constraintSetup [{targetBone {boneName=dyn_amulet}}]
    constraintWeights [1]
   }
   n3 = Look At - Bone: dyn_strap_skirt_05(ComponentPose=@n2, Alpha=1) {
    alphaBoolBlend {blendOption=Linear}
    boneToModify {boneName=dyn_strap_skirt_05}
    lookAtLocation {x=10}
    lookAtTarget {boneReference {boneName=dyn_strap_skirt_06}}
    lookAt_Axis {axis=<1 0>}
    lookUp_Axis {axis=<0 1>}
   }
   n4 = "Constraint - dyn_strap_skirt_06 (PP)"!(ComponentPose=@n3, Alpha=0.1, ConstraintWeights_0=0.5, ConstraintWeights_1:float=0.5) -> (Pose) {
    alphaBoolBlend {blendOption=Linear}
    boneToModify {boneName=dyn_strap_skirt_06}
    constraintSetup [{targetBone {boneName=dyn_strap_skirt_05}}, {targetBone {boneName=dyn_strap_skirt_07}}]
    constraintWeights [1,1]
   }
   n5 = Look At - Bone: dyn_strap_skirt_06(ComponentPose=@n4, Alpha=1) {
    alphaBoolBlend {blendOption=Linear}
    boneToModify {boneName=dyn_strap_skirt_06}
    lookAtLocation {x=10}
    lookAtTarget {boneReference {boneName=dyn_strap_skirt_07}}
    lookAt_Axis {axis=<1 0>}
    lookUp_Axis {axis=<0 1>}
   }
   n6 = "Constraint - dyn_strap_skirt_07 (PPPPP)"!(
    ComponentPose=@n5
    Alpha=1
    ConstraintWeights_0=0.666
    ConstraintWeights_1:float=0.2198
    ConstraintWeights_2:float=0.0743
    ConstraintWeights_3:float=0.021
    ConstraintWeights_4:float=0.0188
   ) -> (Pose) {
    alphaBoolBlend {blendOption=Linear}
    boneToModify {boneName=dyn_strap_skirt_07}
    constraintSetup [
     {targetBone {boneName=pelvis}}
     {targetBone {boneName=spine_02}}
     {targetBone {boneName=spine_01}}
     {targetBone {boneName=spine_04}}
     {targetBone {boneName=spine_03}}
    ]
    constraintWeights [1,1,1,1,1]
   }
   n7 = Control Rig!(
    Source=Component To Local(ComponentPose=@n6)
    Alpha
    jump:bool=false
    parachute:bool=false
    emote:bool=false
    move:bool=false
    FE_Idel:bool=false
    mask:bool=false
    TacticalSprinting:bool=false
    DBNO:bool=false
   ) -> (Pose) {
    alphaBoolBlend {blendOption=Linear}
    controlRigAssetReference {blueprintRigClass=&/F_MED_MoralKore_Line_CtrlRig_post.F_MED_MoralKore_Line_CtrlRig_post_C}
   }
   n8 = "Transform (Modify) Bone - Bone: dyn_skirt_bk_r_02"!(ComponentPose=Local To Component(LocalPose=@n7), bAlphaBoolEnabled:bool=True, Translation="0,0,0", Rotation="0,0,15", Scale="1,1,1") -> (Pose) {
    alphaInputType=Bool
    rotationMode=BMM_Additive
    translationMode=BMM_Additive
    alphaBoolBlend {blendOption=Linear}
    boneToModify {boneName=dyn_skirt_bk_r_02}
    rotation {roll=15}
   }
   n9 = "Transform (Modify) Bone - Bone: dyn_maskholder"!(ComponentPose=@n8, bAlphaBoolEnabled:bool=True, Translation="0,0,0", Rotation="26.161,27.628,-0.000", Scale="1,1,1") -> (Pose) {
    alphaInputType=Bool
    rotationMode=BMM_Additive
    translationMode=BMM_Additive
    alphaBoolBlend {blendOption=Linear}
    boneToModify {boneName=dyn_maskholder}
    rotation {pitch=26.161, yaw=27.628}
   }
   n10 = "Transform (Modify) Bone - Bone: dyn_strap_skirt_05"!(ComponentPose=@n9, bAlphaBoolEnabled:bool=True, Translation="2,0,0", Rotation="0,0,0", Scale="1,1,1") -> (Pose) {
    alphaInputType=Bool
    rotationMode=BMM_Additive
    translationMode=BMM_Additive
    alphaBoolBlend {blendOption=Linear}
    boneToModify {boneName=dyn_strap_skirt_05}
    translation {x=2}
   }
   Output Pose(Result=Component To Local(ComponentPose=@n10)) {layerGroup=None}
   n11 = "Constraint - extra_armor_back (PPP)"!(
    ComponentPose=Local To Component(LocalPose=Copy Pose From Mesh() {bUseAttachedParent=true})
    Alpha=1
    ConstraintWeights_0=0.3
    ConstraintWeights_1:float=0.3
    ConstraintWeights_2:float=0.4
   ) -> (Pose) {
    alphaBoolBlend {blendOption=Linear}
    boneToModify {boneName=extra_armor_back}
    constraintSetup [{targetBone {boneName=spine_05}}, {targetBone {boneName=spine_04}}, {targetBone {boneName=spine_03}}]
    constraintWeights [1,1,1]
   }
   n12 = "Constraint - dyn_strap_skirt_fr_base (PPPPPP)"!(
    ComponentPose=@n11
    Alpha=1
    ConstraintWeights_0=0.7766
    ConstraintWeights_1:float=0.1887
    ConstraintWeights_2:float=0.0218
    ConstraintWeights_3:float=0.0062
    ConstraintWeights_4:float=0.004
    ConstraintWeights_5:float=0.0027
   ) -> (Pose) {
    alphaBoolBlend {blendOption=Linear}
    boneToModify {boneName=dyn_strap_skirt_fr_base}
    constraintSetup [
     {targetBone {boneName=pelvis}}
     {targetBone {boneName=spine_02}}
     {targetBone {boneName=spine_04}}
     {targetBone {boneName=spine_03}}
     {targetBone {boneName=spine_01}}
     {targetBone {boneName=deform_groin_l}}
    ]
    constraintWeights [1,1,1,1,1,1]
   }
   n13 = Control Rig!(Source=Component To Local(ComponentPose=@n12), Alpha, emote=false, move=Get bIsMoving2D()) -> (Pose) {
    alphaBoolBlend {blendOption=Linear}
    controlRigAssetReference {blueprintRigClass=&/F_MED_MoralKore_Line_CtrlRig_pre.F_MED_MoralKore_Line_CtrlRig_pre_C}
   }
   n14 = "Constraint - dyn_skirt_bk_r_base (PPP)"!(
    ComponentPose=Local To Component(LocalPose=@n13)
    Alpha=1
    ConstraintWeights_0=0.4016
    ConstraintWeights_1:float=0.3953
    ConstraintWeights_2:float=0.203
   ) -> (Pose) {
    alphaBoolBlend {blendOption=Linear}
    boneToModify {boneName=dyn_skirt_bk_r_base}
    constraintSetup [{targetBone {boneName=pelvis}}, {targetBone {boneName=deform_glute_r}}, {targetBone {boneName=thigh_twist_01_r}}]
    constraintWeights [1,1,1]
   }
   n15 = "Constraint - dyn_skirt_bk_r_base (R)"(ComponentPose=@n14, Alpha=0.5, ConstraintWeights_0=1) {
    alphaBoolBlend {blendOption=Linear}
    boneToModify {boneName=dyn_skirt_bk_r_base}
    constraintSetup [{transformType=Rotation, targetBone {boneName=thigh_r}}]
    constraintWeights [1]
   }
   n16 = "Constraint - dyn_skirt_bk_l_base (PPP)"!(ComponentPose=@n15, Alpha=1, ConstraintWeights_0=0.4016, ConstraintWeights_1:float=0.3953, ConstraintWeights_2:float=0.203) -> (Pose) {
    alphaBoolBlend {blendOption=Linear}
    boneToModify {boneName=dyn_skirt_bk_l_base}
    constraintSetup [{targetBone {boneName=pelvis}}, {targetBone {boneName=deform_glute_l}}, {targetBone {boneName=thigh_twist_01_l}}]
    constraintWeights [1,1,1]
   }
   n17 = "Constraint - dyn_skirt_bk_l_base (R)"(ComponentPose=@n16, Alpha=0.5, ConstraintWeights_0=1) {
    alphaBoolBlend {blendOption=Linear}
    boneToModify {boneName=dyn_skirt_bk_l_base}
    constraintSetup [{transformType=Rotation, targetBone {boneName=thigh_l}}]
    constraintWeights [1]
   }
   n18 = "Constraint - dyn_skirt_bk_centr_base (PP)"!(ComponentPose=@n17, Alpha=1, ConstraintWeights_0=0.5, ConstraintWeights_1:float=0.5) -> (Pose) {
    alphaBoolBlend {blendOption=Linear}
    boneToModify {boneName=dyn_skirt_bk_centr_base}
    constraintSetup [{targetBone {boneName=dyn_skirt_bk_l_base}}, {targetBone {boneName=dyn_skirt_bk_r_base}}]
    constraintWeights [1,1]
   }
   n19 = "Constraint - extra_knee_pad (PPP)"!(ComponentPose=@n18, Alpha=1, ConstraintWeights_0=0.15, ConstraintWeights_1:float=0.35, ConstraintWeights_2:float=0.5) -> (Pose) {
    alphaBoolBlend {blendOption=Linear}
    boneToModify {boneName=extra_knee_pad}
    constraintSetup [{targetBone {boneName=thigh_l}}, {targetBone {boneName=calf_l}}, {targetBone {boneName=deform_knee_out_l}}]
    constraintWeights [1,1,1]
   }
   n20 = "Constraint - dyn_strap_skirt_07 (P)"(ComponentPose=@n19, Alpha=1, ConstraintWeights_0=1) {
    alphaBoolBlend {blendOption=Linear}
    boneToModify {boneName=dyn_strap_skirt_07}
    constraintSetup [{targetBone {boneName=dyn_strap_skirt_back_base}}]
    constraintWeights [1]
   }
   n21 = RigidBody!(ComponentPose=@n20, bAlphaBoolEnabled:bool, OverrideWorldGravity:Vector="0,0,-980", ExternalForce) -> (Pose) {
    alphaInputType=Bool
    bClampLinearTranslationLimitToRefPose=true
    bFreezeIncomingPoseOnStart=true
    bOverrideWorldGravity=true
    bTransferBoneVelocities=true
    overridePhysicsAsset=&/F_MED_MoralKore_Line_Physics
    simulationSpace=BaseBoneSpace
    alphaBoolBlend {blendInTime=0.3, blendOption=Linear, blendOutTime=0.3}
    baseBoneRef {boneName=pelvis}
    componentAppliedLinearAccClamp=<3000 3000 3000>
    componentLinearAccScale=<0.6 0.6 0.6>
    componentLinearVelScale=<0.4 0.4 0.4>
    overrideWorldGravity {z=-980}
    simSpaceSettings {maxAngularAcceleration=3000, maxAngularVelocity=3000, maxLinearAcceleration=3000, maxLinearVelocity=3000, worldAlpha=0.5}
   }
   n22 = RigidBody(ComponentPose=@n21, OverrideWorldGravity="0,0,-980") {
    alphaInputType=Bool
    bClampLinearTranslationLimitToRefPose=true
    bFreezeIncomingPoseOnStart=true
    bOverrideWorldGravity=true
    bTransferBoneVelocities=true
    overridePhysicsAsset=&/F_MED_MoralKore_Line_Physics_FE
    simulationSpace=BaseBoneSpace
    alphaBoolBlend {blendInTime=0.3, blendOption=Linear, blendOutTime=0.3}
    baseBoneRef {boneName=pelvis}
    componentAppliedLinearAccClamp=<3000 3000 3000>
    componentLinearAccScale=<0.6 0.6 0.6>
    componentLinearVelScale=<0.4 0.4 0.4>
    overrideWorldGravity {z=-980}
    simSpaceSettings {maxAngularAcceleration=3000, maxAngularVelocity=3000, maxLinearAcceleration=3000, maxLinearVelocity=3000, worldAlpha=0.5}
   }
   n23 = "Transform (Modify) Bone - Bone: dyn_strap_skirt_04"!(ComponentPose=@n22, bAlphaBoolEnabled:bool=True, Translation="0,0,-1", Rotation="0,0,0", Scale="1,1,1") -> (Pose) {
    alphaInputType=Bool
    translationMode=BMM_Additive
    translationSpace=BCS_BoneSpace
    alphaBoolBlend {blendOption=Linear}
    boneToModify {boneName=dyn_strap_skirt_04}
    translation {z=-1}
   }
   n24 = "Transform (Modify) Bone - Bone: dyn_strap_skirt_06"!(ComponentPose=@n23, bAlphaBoolEnabled:bool=True, Translation="0,0,1", Rotation="0,0,0", Scale="1,1,1") -> (Pose) {
    alphaInputType=Bool
    translationMode=BMM_Additive
    translationSpace=BCS_BoneSpace
    alphaBoolBlend {blendOption=Linear}
    boneToModify {boneName=dyn_strap_skirt_06}
    translation {z=1}
   }
   Save cached pose 'post_dyn'(Pose=Component To Local(ComponentPose=@n24)) {cacheName=post_dyn}
   n25 = "Transform (Modify) Bone - Bone: dyn_amulet"(
    ComponentPose=Local To Component(LocalPose=Use cached pose 'maskOff_crouch'())
    Alpha=0.8
    Translation="-0.093,0.007,0.784"
    Rotation="8.562,-0.268,-1.456"
    Scale="1,1,1"
   ) {
    rotationMode=BMM_Additive
    rotationSpace=BCS_BoneSpace
    translationMode=BMM_Additive
    translationSpace=BCS_BoneSpace
    alphaBoolBlend {blendOption=Linear}
    boneToModify {boneName=dyn_amulet}
    rotation=<rot 8.562 -0.2679 -1.456>
    translation=<-0.093 0.007 0.784>
   }
   n26 = Component To Local(ComponentPose=@n25)
   n27 = Use cached pose 'maskOff_crouch'()
   n28 = "Constraint - dyn_strap_skirt_04 (PP)"!(ComponentPose=Local To Component(LocalPose=Use cached pose 'mask_crouch'()), Alpha=1, ConstraintWeights_0=0.2, ConstraintWeights_1:float=0.8) -> (Pose) {
    alphaBoolBlend {blendOption=Linear}
    boneToModify {boneName=dyn_strap_skirt_04}
    constraintSetup [{targetBone {boneName=thigh_l}}, {targetBone {boneName=pelvis}}]
    constraintWeights [1,1]
   }
   n29 = "Transform (Modify) Bone - Bone: dyn_strap_skirt_04"!(ComponentPose=@n28, Alpha:float=0.8, Translation="6.744,0.072,1.025", Rotation="-20.176,-42.309,8.763", Scale="1,1,1") -> (Pose) {
    rotationMode=BMM_Additive
    rotationSpace=BCS_BoneSpace
    translationMode=BMM_Additive
    translationSpace=BCS_BoneSpace
    alphaBoolBlend {blendOption=Linear}
    boneToModify {boneName=dyn_strap_skirt_04}
    rotation=<rot -20.176 -42.3089 8.763>
    translation=<6.744 0.072 1.025>
   }
   n30 = Component To Local(ComponentPose=@n29)
   n31 = Use cached pose 'mask_crouch'()
   n32 = "Transform (Modify) Bone - Bone: dyn_strap_skirt_04"(
    ComponentPose=Local To Component(LocalPose=Blend Poses by bool(BlendPose_0=@n30, BlendPose_1=@n31, BlendTime_0=0.3, BlendTime_1=0.3, bActiveValue=true) {blendType=Linear})
    Alpha=0.8
    Translation="-0.000,-0.057,-0.702"
    Rotation="0,0,0"
    Scale="1,1,1"
   ) {
    rotationMode=BMM_Additive
    rotationSpace=BCS_BoneSpace
    translationMode=BMM_Additive
    translationSpace=BCS_BoneSpace
    alphaBoolBlend {blendOption=Linear}
    boneToModify {boneName=dyn_strap_skirt_04}
    translation {y=-0.057, z=-0.702}
   }
   Save cached pose 'mask'(Pose=Blend Poses by bool(
     BlendPose_0=Component To Local(ComponentPose=@n32)
     BlendPose_1=Blend Poses by bool(BlendPose_0=@n26, BlendPose_1=@n27, BlendTime_0=0.3, BlendTime_1=0.3, bActiveValue=true) {blendType=Linear}
     BlendTime_0=0.8
     BlendTime_1=0.8
     bActiveValue=true
    ) {blendType=Linear}) {cacheName=mask}
   n33 = "Transform (Modify) Bone - Bone: dyn_strap_skirt_02"!(
    ComponentPose=Local To Component(LocalPose=Use cached pose 'post_dyn'())
    bAlphaBoolEnabled:bool=True
    Translation="-2,1.5,0"
    Rotation="0,0,0"
    Scale="1,1,1"
   ) -> (Pose) {
    alphaInputType=Bool
    rotationMode=BMM_Additive
    rotationSpace=BCS_BoneSpace
    translationMode=BMM_Additive
    translationSpace=BCS_BoneSpace
    alphaBoolBlend {blendOption=Linear}
    boneToModify {boneName=dyn_strap_skirt_02}
    translation=<-2 1.5>
   }
   n34 = "Transform (Modify) Bone - Bone: dyn_strap_skirt_04"!(ComponentPose=@n33, bAlphaBoolEnabled:bool=True, Translation="0.988,3.429,0.791", Rotation="0.027,-10.000,4.998", Scale="1,1,1") -> (Pose) {
    alphaInputType=Bool
    rotationMode=BMM_Additive
    rotationSpace=BCS_BoneSpace
    translationMode=BMM_Additive
    translationSpace=BCS_BoneSpace
    alphaBoolBlend {blendOption=Linear}
    boneToModify {boneName=dyn_strap_skirt_04}
    rotation=<rot 0.027 -9.9999 4.9982>
    translation=<0.988 3.429 0.791>
   }
   n35 = "Transform (Modify) Bone - Bone: dyn_strap_skirt_05"(ComponentPose=@n34, bAlphaBoolEnabled=True, Translation="-4.205,-5.511,1", Rotation="26.923,57.421,-23.063", Scale="1,1,1") {
    alphaInputType=Bool
    rotationMode=BMM_Additive
    rotationSpace=BCS_BoneSpace
    translationMode=BMM_Additive
    translationSpace=BCS_BoneSpace
    alphaBoolBlend {blendOption=Linear}
    boneToModify {boneName=dyn_strap_skirt_05}
    rotation=<rot 26.923 57.4211 -23.0629>
    translation=<-4.205 -5.511 1>
   }
   Save cached pose 'mask_crouch'(Pose=Blend Poses by bool(
     BlendPose_0=Component To Local(ComponentPose=@n35)
     BlendPose_1=Use cached pose 'post_dyn'()
     BlendTime_0=0.3
     BlendTime_1=0.3
     bActiveValue=true
    ) {blendType=Linear}) {cacheName=mask_crouch}
   n36 = "Transform (Modify) Bone - Bone: dyn_strap_skirt_01"!(
    ComponentPose=Local To Component(LocalPose=Use cached pose 'post_dyn'())
    bAlphaBoolEnabled:bool=True
    Translation="-0.005,-0.061,0.428"
    Rotation="3.290,20.222,-9.181"
    Scale="1,1,1"
   ) -> (Pose) {
    alphaInputType=Bool
    rotationMode=BMM_Additive
    rotationSpace=BCS_BoneSpace
    translationMode=BMM_Additive
    translationSpace=BCS_BoneSpace
    alphaBoolBlend {blendOption=Linear}
    boneToModify {boneName=dyn_strap_skirt_01}
    rotation=<rot 3.2901 20.2219 -9.1812>
    translation=<-0.005 -0.061 0.428>
   }
   n37 = "Constraint - dyn_strap_skirt_04 (P)"(ComponentPose=@n36, Alpha=0.9, ConstraintWeights_0=1) {
    alphaBoolBlend {blendOption=Linear}
    boneToModify {boneName=dyn_strap_skirt_04}
    constraintSetup [{targetBone {boneName=pelvis}}]
    constraintWeights [1]
   }
   n38 = "Constraint - dyn_amulet (P)"(ComponentPose=@n37, Alpha=0.9, ConstraintWeights_0=1) {
    alphaBoolBlend {blendOption=Linear}
    boneToModify {boneName=dyn_amulet}
    constraintSetup [{targetBone {boneName=dyn_strap_skirt_04}}]
    constraintWeights [1]
   }
   n39 = "Transform (Modify) Bone - Bone: dyn_strap_skirt_04"(ComponentPose=@n38, bAlphaBoolEnabled=True, Translation="5.194,3.725,0.567", Rotation="-23.855,-13.349,24.541", Scale="1,1,1") {
    alphaInputType=Bool
    rotationMode=BMM_Additive
    rotationSpace=BCS_BoneSpace
    translationMode=BMM_Additive
    translationSpace=BCS_BoneSpace
    alphaBoolBlend {blendOption=Linear}
    boneToModify {boneName=dyn_strap_skirt_04}
    rotation=<rot -23.855 -13.349 24.541>
    translation=<5.194 3.725 0.567>
   }
   Save cached pose 'maskOff_crouch'(Pose=Blend Poses by bool(
     BlendPose_0=Component To Local(ComponentPose=@n39)
     BlendPose_1=Use cached pose 'post_dyn'()
     BlendTime_0=0.3
     BlendTime_1=0.3
     bActiveValue=true
    ) {blendType=Linear}) {cacheName=maskOff_crouch}
   Use cached pose 'mask'()
  ]}
 EventGraph {nodes [
   Event BlueprintUpdateAnimation()
   > n1 = Sequence!(execute) -> (then_0, then_1, then_2:Exec, then_3:Exec, then_4:Exec, then_5:Exec)
   > n2 = Gravity Override State Function DEPRECATED(
    CharacterPartAnimInstance=Self-Reference()
    GameTimeInSeconds=GetGameTimeInSeconds()
    GravityOverrideParams=Get Gravity Override Params()
    RBAN Params=Get RBAN Params()
   )
   > Set Gravity(Gravity:Vector=@n2) -> (Output_Get:Vector)
   n3 = Get bIsInFrontEnd()
   n4 = AND Boolean(A=@n3, B=NOT Boolean(A=Get bIsPlayingEmote()))
   Set FE_Idel(execute=@n1.then_5, FE_Idel:bool=@n4) -> (Output_Get:bool)
   > Set dyn(dyn:bool=NOT Boolean(A=@n4)) -> (Output_Get:bool)
   n5 = Get bRigidBodyEnabled()
   > Set dyn_Disabled(dyn_Disabled:bool=AND Boolean(A=NOT Boolean(A=@n5), B=@n3)) -> (Output_Get:bool)
   > Set Is Tactical Sprinting(Is Tactical Sprinting:bool=Get bIsTacticalSprinting()) -> (Output_Get:bool)
   Set jump(execute=@n1.then_3, jump:bool=OR Boolean(A=Get bIsJumping(), B=Get bIsFalling())) -> (Output_Get:bool)
   Set mask(execute=@n1.then_4, mask:bool=CharacterItemDefinitionHasAnyMetaTag(Tags=Make Array("[0]"=(TagName="Cosmetics.Variant.Property.Mat1")))) -> (Output_Get:bool)
   > Set parachute(parachute:bool=Get bIsParachuteOpened()) -> (Output_Get:bool)
   n6 = Get bIsMoving2D()
   Set crouch_not_move(execute=@n1.then_1, crouch_not_move:bool=AND Boolean(A=Get bIsCrouching(), B=NOT Boolean(A=@n6))) -> (Output_Get:bool)
   Set crouch_move(execute=@n1.then_2, crouch_move:bool=AND Boolean(A=Get bIsCrouching(), B=Get bIsMoving2D())) -> (Output_Get:bool)
   Event OnCharacterCustomizationLoaded()
   > Set BindPose(BindPose:bool=false) -> (Output_Get:bool)
  ]}
]
variables [
 BindPose {type=boolean, default=true}
 crouch_move {type=boolean}
 crouch_not_move {type=boolean}
 dyn {type=boolean}
 dyn_Disabled {type=boolean}
 FE_Idel {type=boolean}
 Gravity {type=Vector}
 Gravity Override Params {type=GravityOverrideParamsStruct}
 Is Tactical Sprinting {type=boolean}
 jump {type=boolean}
 mask {type=boolean}
 parachute {type=boolean}
 RBAN Params {
  type=RBANParamsStruct
  default {
   dBNO {simSpaceSettings {
     maxAngularAcceleration=1000
     maxAngularVelocity=3
     maxLinearAcceleration=300
     maxLinearVelocity=700
     velocityScaleZ=0
     externalLinearDragV=<0 0.5 0>
    }}
   dBNO_Carried {simSpaceSettings {
     maxAngularAcceleration=1000
     maxAngularVelocity=3
     maxLinearAcceleration=300
     maxLinearVelocity=700
     velocityScaleZ=0
     externalLinearDragV=<0 0.5 0>
    }}
   emoteOrMelee {simSpaceSettings {
     maxAngularAcceleration=1000
     maxAngularVelocity=3
     maxLinearAcceleration=300
     maxLinearVelocity=700
     velocityScaleZ=0
     externalLinearDragV=<0 0.5 0>
    }}
   front_End {simSpaceSettings {
     maxAngularAcceleration=1000
     maxAngularVelocity=3
     maxLinearAcceleration=300
     maxLinearVelocity=700
     velocityScaleZ=0
     externalLinearDragV=<0 0.5 0>
    }}
   onGround_Crouch_Moving {simSpaceSettings {
     maxAngularAcceleration=1000
     maxAngularVelocity=3
     maxLinearAcceleration=300
     maxLinearVelocity=700
     velocityScaleZ=0
     externalLinearDragV=<0 0.5 0>
    }}
   onGround_Falling {simSpaceSettings {
     maxAngularAcceleration=1000
     maxAngularVelocity=3
     maxLinearAcceleration=300
     maxLinearVelocity=700
     velocityScaleZ=0
     externalLinearDragV=<0 0.5 0>
    }}
   onGround_Moving {simSpaceSettings {
     maxAngularAcceleration=1000
     maxAngularVelocity=3
     maxLinearAcceleration=300
     maxLinearVelocity=700
     velocityScaleZ=0
     externalLinearDragV=<0 0.5 0>
    }}
   onGround_Standing {simSpaceSettings {
     maxAngularAcceleration=1000
     maxAngularVelocity=3
     maxLinearAcceleration=300
     maxLinearVelocity=700
     velocityScaleZ=0
     externalLinearDragV=<0 0.5 0>
    }}
   rail_Grinding {simSpaceSettings {
     maxAngularAcceleration=1000
     maxAngularVelocity=3
     maxLinearAcceleration=300
     maxLinearVelocity=700
     velocityScaleZ=0
     worldAlpha=0.1
     externalLinearDragV=<0 0.5 0>
    }}
   riding_Moving {simSpaceSettings {
     maxAngularAcceleration=1000
     maxAngularVelocity=3
     maxLinearAcceleration=300
     maxLinearVelocity=700
     velocityScaleZ=0
     externalLinearDragV=<0 0.5 0>
    }}
   riding_Moving_Targeting {simSpaceSettings {
     maxAngularAcceleration=1000
     maxAngularVelocity=3
     maxLinearAcceleration=300
     maxLinearVelocity=700
     velocityScaleZ=0
     externalLinearDragV=<0 0.5 0>
    }}
   riding_Standing {simSpaceSettings {
     maxAngularAcceleration=1000
     maxAngularVelocity=3
     maxLinearAcceleration=300
     maxLinearVelocity=700
     velocityScaleZ=0
     externalLinearDragV=<0 0.5 0>
    }}
   skydive_DivingDown {simSpaceSettings {
     maxAngularAcceleration=1000
     maxAngularVelocity=3
     maxLinearAcceleration=300
     maxLinearVelocity=700
     velocityScaleZ=0
     externalLinearDragV=<0 0.5 0>
    }}
   skydive_FloatingDown {
    componentLinearAccScale=<0 0.5 0>
    componentLinearVelScale=<0 0.5 0>
    simSpaceSettings {
     maxAngularAcceleration=1000
     maxAngularVelocity=3
     maxLinearAcceleration=300
     maxLinearVelocity=700
     velocityScaleZ=0
     externalLinearDragV=<0 0.5 0>
    }
   }
   skydive_GoingUp {simSpaceSettings {
     maxAngularAcceleration=1000
     maxAngularVelocity=3
     maxLinearAcceleration=300
     maxLinearVelocity=700
     velocityScaleZ=0
     externalLinearDragV=<0 0.5 0>
    }}
   skydive_Parachuting {
    componentLinearAccScale=<0 0.1 0.3>
    componentLinearVelScale=<0 0.1 0.3>
    simSpaceSettings {
     maxAngularAcceleration=1000
     maxAngularVelocity=3
     maxLinearAcceleration=300
     maxLinearVelocity=700
     velocityScaleZ=0
     externalLinearDragV=<0 0.5 0>
    }
   }
   skydive_Surfing {simSpaceSettings {
     maxAngularAcceleration=1000
     maxAngularVelocity=3
     maxLinearAcceleration=300
     maxLinearVelocity=700
     velocityScaleZ=0
     externalLinearDragV=<0 0.5 0>
    }}
   swimming {simSpaceSettings {
     maxAngularAcceleration=1000
     maxAngularVelocity=3
     maxLinearAcceleration=300
     maxLinearVelocity=700
     velocityScaleZ=0
     externalLinearDragV=<0 0.5 0>
    }}
   tactical_Sprint {simSpaceSettings {
     maxAngularAcceleration=1000
     maxAngularVelocity=3
     maxLinearAcceleration=300
     maxLinearVelocity=700
     velocityScaleZ=0
     externalLinearDragV=<0 0.5 0>
    }}
   ziplining {simSpaceSettings {
     maxAngularAcceleration=1000
     maxAngularVelocity=3
     maxLinearAcceleration=300
     maxLinearVelocity=700
     velocityScaleZ=0
     externalLinearDragV=<0 0.5 0>
    }}
  }
 }
]
