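# Physics asset for the MoralKore Line female medium face accessory, providing capsule bodies for dangling hair, jewelry and collar bones.
# Front, side and back hair chains and left/right jewelry link through base bones to the head, with limited swing and twist except the free back-hair base.
# Collar bones link through extra_collar bones to spine_05 with free angular motion.
# Neck, clavicle and upper-arm capsules have no constraints, apparently serving as colliders.
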PhysicsAsset /BRCosmetics/Characters/Player/Female/Medium/Bodies/F_MED_MoralKore_Line/Meshes/Parts/F_MED_MoralKore_Line_FaceAcc_Physics_Clove_FE
bodies {
 clavicle_l [{shapeType=Capsule, length=6, radius=6, shapeName=clavicle_l_capsule, center=<1.8251 0 0>, rotation=<rot 0 -90 90>}]
 clavicle_r [{shapeType=Capsule, length=5, radius=7.5, shapeName=clavicle_r_capsule, center=<-5 0 0>, rotation=<rot 0 90 90>}]
 dyn_collar_02_l [{shapeType=Capsule, length=5, radius=3, shapeName=dyn_collar_02_l_capsule, center=<3 0 0>, rotation=<rot 0 -90 0>}]
 dyn_collar_02_r [{shapeType=Capsule, length=5, radius=3, shapeName=dyn_collar_02_r_capsule, center=<3 0 0>, rotation=<rot 0 -90 0>}]
 dyn_collar_04_l [{shapeType=Capsule, length=5, radius=2, shapeName=dyn_collar_04_l_capsule, center=<5 0 0>, rotation=<rot 0 -90 0>}]
 dyn_collar_04_r [{shapeType=Capsule, length=5, radius=2, shapeName=dyn_collar_04_r_capsule, center=<5 0 0>, rotation=<rot 0 -90 0>}]
 dyn_collar_06_l [{shapeType=Capsule, length=5, radius=1, shapeName=dyn_collar_06_l_capsule, center=<5 0 0>, rotation=<rot 0 -90 0>}]
 dyn_collar_06_r [{shapeType=Capsule, length=5, radius=1, shapeName=dyn_collar_06_r_capsule, center=<5 0 0>, rotation=<rot 0 -90 0>}]
 dyn_collar_m_l [{shapeType=Capsule, length=5, radius=3, shapeName=dyn_collar_m_l_capsule, center=<3 0 0>, rotation=<rot 0 -90 0>}]
 dyn_hair_back [{shapeType=Capsule, length=2, radius=4, shapeName=dyn_hair_back_capsule, center=<6 0 0>, rotation=<rot 0 90 90>}]
 dyn_hair_back_base [{shapeType=Capsule, length=0.505, radius=0.505, shapeName=dyn_hair_back_base_capsule}]
 dyn_hair_fr_l_01 [{shapeType=Capsule, length=1, radius=1, shapeName=dyn_hair_fr_l_01_capsule, center=<2 0 0>, rotation=<rot 0 0 90>}]
 dyn_hair_fr_l_02 [{shapeType=Capsule, length=1, radius=0.5, shapeName=dyn_hair_fr_l_02_capsule, center=<2 0 0>, rotation=<rot 0 0 90>}]
 dyn_hair_fr_l_base [{shapeType=Capsule, length=1, radius=1, shapeName=dyn_hair_fr_l_base_capsule}]
 dyn_hair_fr_r_01 [{shapeType=Capsule, length=1, radius=1, shapeName=dyn_hair_fr_r_01_capsule, center=<2 0 0>, rotation=<rot 0 0 90>}]
 dyn_hair_fr_r_02 [{shapeType=Capsule, length=1, radius=0.5, shapeName=dyn_hair_fr_r_02_capsule, center=<2 0 0>, rotation=<rot 0 0 90>}]
 dyn_hair_fr_r_base [{shapeType=Capsule, length=1, radius=1, shapeName=dyn_hair_fr_r_base_capsule}]
 dyn_hair_fr_side_l_01 [{shapeType=Capsule, length=1, radius=1, shapeName=dyn_hair_fr_side_l_01_capsule, center=<2 0 0>, rotation=<rot 0 0 90>}]
 dyn_hair_fr_side_l_02 [{shapeType=Capsule, length=1, radius=0.5, shapeName=dyn_hair_fr_side_l_02_capsule, center=<2 0 0>, rotation=<rot 0 0 90>}]
 dyn_hair_fr_side_l_base [{shapeType=Capsule, length=1, radius=1, shapeName=dyn_hair_fr_side_l_base_capsule}]
 dyn_hair_fr_side_r_01 [{shapeType=Capsule, length=1, radius=1, shapeName=dyn_hair_fr_side_r_01_capsule, center=<2 0 0>, rotation=<rot 0 0 90>}]
 dyn_hair_fr_side_r_02 [{shapeType=Capsule, length=1, radius=1, shapeName=dyn_hair_fr_side_r_02_capsule, center=<2 0 0>, rotation=<rot 0 0 90>}]
 dyn_hair_fr_side_r_base [{shapeType=Capsule, length=1, radius=1, shapeName=dyn_hair_fr_side_r_base_capsule}]
 dyn_hair_side_l_01 [{shapeType=Capsule, length=1, radius=1, shapeName=dyn_hair_side_l_01_capsule, center=<2 0 0>, rotation=<rot 90 0 0>}]
 dyn_hair_side_l_02 [{shapeType=Capsule, length=2, radius=1, shapeName=dyn_hair_side_l_02_capsule, center=<2 0 0>, rotation=<rot 90 0 0>}]
 dyn_hair_side_l_03 [{shapeType=Capsule, length=6, radius=1.2, shapeName=dyn_hair_side_l_03_capsule, center=<5 0 0>, rotation=<rot 90 0 0>}]
 dyn_hair_side_l_base [{shapeType=Capsule, length=1, radius=1, shapeName=dyn_hair_side_l_base_capsule}]
 dyn_hair_side_r_01 [{shapeType=Capsule, length=1, radius=1, shapeName=dyn_hair_side_r_01_capsule, center=<2 0 0>, rotation=<rot 90 0 0>}]
 dyn_hair_side_r_02 [{shapeType=Capsule, length=2, radius=1, shapeName=dyn_hair_side_r_02_capsule, center=<2 0 0>, rotation=<rot 90 0 0>}]
 dyn_hair_side_r_03 [{shapeType=Capsule, length=6, radius=1, shapeName=dyn_hair_side_r_03_capsule, center=<5 0 0>, rotation=<rot 90 0 0>}]
 dyn_hair_side_r_base [{shapeType=Capsule, length=1, radius=1, shapeName=dyn_hair_side_r_base_capsule}]
 dyn_jewelry_l [{shapeType=Capsule, length=2, radius=0.5, shapeName=dyn_jewelry_l_capsule, center=<2.5 0 0>, rotation=<rot 90 0 0>}]
 dyn_jewelry_l_base [{shapeType=Capsule, length=1, radius=1, shapeName=dyn_jewelry_l_base_capsule}]
 dyn_jewelry_r [{shapeType=Capsule, length=2, radius=0.5, shapeName=dyn_jewelry_r_capsule, center=<2.5 0 0>, rotation=<rot 90 0 0>}]
 dyn_jewelry_r_base [{shapeType=Capsule, length=1, radius=1, shapeName=dyn_jewelry_r_base_capsule}]
 extra_collar_02_l [{shapeType=Capsule, length=0.505, radius=0.505, shapeName=extra_collar_02_l_capsule, rotation=<rot 0 -90 0>}]
 extra_collar_02_r [{shapeType=Capsule, length=0.505, radius=0.505, shapeName=extra_collar_02_r_capsule, rotation=<rot 0 -90 0>}]
 extra_collar_04_l [{shapeType=Capsule, length=0.505, radius=0.505, shapeName=extra_collar_04_l_capsule, rotation=<rot 0 -90 0>}]
 extra_collar_04_r [{shapeType=Capsule, length=0.505, radius=0.505, shapeName=extra_collar_04_r_capsule, rotation=<rot 0 -90 0>}]
 extra_collar_06_l [{shapeType=Capsule, length=0.505, radius=0.505, shapeName=extra_collar_06_l_capsule, rotation=<rot 0 -90 0>}]
 extra_collar_06_r [{shapeType=Capsule, length=0.505, radius=0.505, shapeName=extra_collar_06_r_capsule, rotation=<rot 0 -90 0>}]
 extra_collar_m_l [{shapeType=Capsule, length=0.505, radius=0.505, shapeName=extra_collar_m_l_capsule, rotation=<rot 0 -90 0>}]
 head [
  {shapeType=Capsule, length=10, radius=6, center=<2.2264 0 0>, rotation=<rot 0 -90 -90>}
  {shapeType=Capsule, length=5, radius=5.5, shapeName=head_capsule, center=<4.3825 5.2804 0>, rotation=<rot 0 90 90>}
 ]
 neck_01 [{shapeType=Capsule, length=5, radius=5, shapeName=neck_01_capsule, rotation=<rot 90 -90 90>}]
 spine_05 [{shapeType=Capsule, length=3.5, radius=9, shapeName=spine_05_capsule, rotation=<rot 0 90 0>}]
 upperarm_l [{shapeType=Capsule, length=21.75, radius=7, shapeName=upperarm_l_capsule, center=<12.3273 0 0>, rotation=<rot 90 0 -0.0002>}]
 upperarm_r [
  {shapeType=Capsule, length=21.75, radius=3.5, shapeName=upperarm_r_capsule, center=<-12.3273 0 0>, rotation=<rot 90 180 179.9995>}
  {
   shapeType=Capsule
   length=5
   radius=4.5
   shapeName=clavicle_r_capsule
   center=<-4.5816 -1.1017 -1.1257>
   rotation=<rot -89.857 179.8207 -89.8209>
  }
 ]
}
constraints [
 {
  bone1Name=dyn_hair_back_base
  bone2Name=head
  swing1LimitDegrees=45
  swing1Motion=Free
  swing2LimitDegrees=45
  swing2Motion=Free
  twistLimitDegrees=45
  twistMotion=Free
 }
 {bone1Name=dyn_hair_back, bone2Name=dyn_hair_back_base, swing1LimitDegrees=30, swing2LimitDegrees=30, twistLimitDegrees=5}
 {bone1Name=dyn_hair_fr_l_base, bone2Name=head, swing1LimitDegrees=15, swing2LimitDegrees=15, twistLimitDegrees=5}
 {bone1Name=dyn_hair_fr_l_01, bone2Name=dyn_hair_fr_l_base, swing1LimitDegrees=10, swing2LimitDegrees=20, twistLimitDegrees=5}
 {bone1Name=dyn_hair_fr_l_02, bone2Name=dyn_hair_fr_l_01, swing1LimitDegrees=10, swing2LimitDegrees=20, twistLimitDegrees=5}
 {bone1Name=dyn_hair_fr_r_base, bone2Name=head, swing1LimitDegrees=15, swing2LimitDegrees=15, twistLimitDegrees=5}
 {bone1Name=dyn_hair_fr_r_01, bone2Name=dyn_hair_fr_r_base, swing1LimitDegrees=10, swing2LimitDegrees=20, twistLimitDegrees=5}
 {bone1Name=dyn_hair_fr_r_02, bone2Name=dyn_hair_fr_r_01, swing1LimitDegrees=10, swing2LimitDegrees=20, twistLimitDegrees=5}
 {bone1Name=dyn_hair_fr_side_l_base, bone2Name=head, swing1LimitDegrees=15, swing2LimitDegrees=15, twistLimitDegrees=5}
 {bone1Name=dyn_hair_fr_side_l_01, bone2Name=dyn_hair_fr_side_l_base, swing1LimitDegrees=10, swing2LimitDegrees=20, twistLimitDegrees=5}
 {bone1Name=dyn_hair_fr_side_l_02, bone2Name=dyn_hair_fr_side_l_01, swing1LimitDegrees=10, swing2LimitDegrees=20, twistLimitDegrees=5}
 {bone1Name=dyn_hair_fr_side_r_base, bone2Name=head, swing1LimitDegrees=15, swing2LimitDegrees=15, twistLimitDegrees=5}
 {bone1Name=dyn_hair_fr_side_r_01, bone2Name=dyn_hair_fr_side_r_base, swing1LimitDegrees=10, swing2LimitDegrees=20, twistLimitDegrees=5}
 {bone1Name=dyn_hair_fr_side_r_02, bone2Name=dyn_hair_fr_side_r_01, swing1LimitDegrees=10, swing2LimitDegrees=20, twistLimitDegrees=5}
 {bone1Name=dyn_hair_side_l_base, bone2Name=head, swing1LimitDegrees=15, swing2LimitDegrees=15, twistLimitDegrees=5}
 {bone1Name=dyn_hair_side_l_01, bone2Name=dyn_hair_side_l_base, swing1LimitDegrees=20, swing2LimitDegrees=15, twistLimitDegrees=5}
 {bone1Name=dyn_hair_side_l_02, bone2Name=dyn_hair_side_l_01, swing1LimitDegrees=10, swing2LimitDegrees=20, twistLimitDegrees=5}
 {bone1Name=dyn_hair_side_l_03, bone2Name=dyn_hair_side_l_02, swing1LimitDegrees=15, swing2LimitDegrees=30, twistLimitDegrees=5}
 {bone1Name=dyn_hair_side_r_base, bone2Name=head, swing1LimitDegrees=15, swing2LimitDegrees=15, twistLimitDegrees=5}
 {bone1Name=dyn_hair_side_r_01, bone2Name=dyn_hair_side_r_base, swing1LimitDegrees=20, swing2LimitDegrees=15, twistLimitDegrees=5}
 {bone1Name=dyn_hair_side_r_02, bone2Name=dyn_hair_side_r_01, swing1LimitDegrees=10, swing2LimitDegrees=20, twistLimitDegrees=5}
 {bone1Name=dyn_hair_side_r_03, bone2Name=dyn_hair_side_r_02, swing1LimitDegrees=15, swing2LimitDegrees=30, twistLimitDegrees=5}
 {bone1Name=dyn_jewelry_l_base, bone2Name=head, swing1LimitDegrees=15, swing2LimitDegrees=15, twistLimitDegrees=5}
 {bone1Name=dyn_jewelry_l, bone2Name=dyn_jewelry_l_base, swing1LimitDegrees=30, swing2LimitDegrees=30, twistLimitDegrees=5}
 {bone1Name=dyn_jewelry_r_base, bone2Name=head, swing1LimitDegrees=15, swing2LimitDegrees=15, twistLimitDegrees=5}
 {bone1Name=dyn_jewelry_r, bone2Name=dyn_jewelry_r_base, swing1LimitDegrees=30, swing2LimitDegrees=30, twistLimitDegrees=5}
 {
  bone1Name=dyn_collar_m_l
  bone2Name=extra_collar_m_l
  swing1LimitDegrees=45
  swing1Motion=Free
  swing2LimitDegrees=45
  swing2Motion=Free
  twistLimitDegrees=45
  twistMotion=Free
 }
 {
  bone1Name=dyn_collar_06_r
  bone2Name=extra_collar_06_r
  swing1LimitDegrees=45
  swing1Motion=Free
  swing2LimitDegrees=45
  swing2Motion=Free
  twistLimitDegrees=45
  twistMotion=Free
 }
 {
  bone1Name=dyn_collar_06_l
  bone2Name=extra_collar_06_l
  swing1LimitDegrees=45
  swing1Motion=Free
  swing2LimitDegrees=45
  swing2Motion=Free
  twistLimitDegrees=45
  twistMotion=Free
 }
 {
  bone1Name=dyn_collar_04_r
  bone2Name=extra_collar_04_r
  swing1LimitDegrees=45
  swing1Motion=Free
  swing2LimitDegrees=45
  swing2Motion=Free
  twistLimitDegrees=45
  twistMotion=Free
 }
 {
  bone1Name=dyn_collar_04_l
  bone2Name=extra_collar_04_l
  swing1LimitDegrees=45
  swing1Motion=Free
  swing2LimitDegrees=45
  swing2Motion=Free
  twistLimitDegrees=45
  twistMotion=Free
 }
 {
  bone1Name=dyn_collar_02_r
  bone2Name=extra_collar_02_r
  swing1LimitDegrees=45
  swing1Motion=Free
  swing2LimitDegrees=45
  swing2Motion=Free
  twistLimitDegrees=45
  twistMotion=Free
 }
 {
  bone1Name=dyn_collar_02_l
  bone2Name=extra_collar_02_l
  swing1LimitDegrees=45
  swing1Motion=Free
  swing2LimitDegrees=45
  swing2Motion=Free
  twistLimitDegrees=45
  twistMotion=Free
 }
 {
  bone1Name=extra_collar_m_l
  bone2Name=spine_05
  swing1LimitDegrees=45
  swing1Motion=Free
  swing2LimitDegrees=45
  swing2Motion=Free
  twistLimitDegrees=45
  twistMotion=Free
 }
 {
  bone1Name=extra_collar_06_r
  bone2Name=spine_05
  swing1LimitDegrees=45
  swing1Motion=Free
  swing2LimitDegrees=45
  swing2Motion=Free
  twistLimitDegrees=45
  twistMotion=Free
 }
 {
  bone1Name=extra_collar_06_l
  bone2Name=spine_05
  swing1LimitDegrees=45
  swing1Motion=Free
  swing2LimitDegrees=45
  swing2Motion=Free
  twistLimitDegrees=45
  twistMotion=Free
 }
 {
  bone1Name=extra_collar_04_r
  bone2Name=spine_05
  swing1LimitDegrees=45
  swing1Motion=Free
  swing2LimitDegrees=45
  swing2Motion=Free
  twistLimitDegrees=45
  twistMotion=Free
 }
 {
  bone1Name=extra_collar_04_l
  bone2Name=spine_05
  swing1LimitDegrees=45
  swing1Motion=Free
  swing2LimitDegrees=45
  swing2Motion=Free
  twistLimitDegrees=45
  twistMotion=Free
 }
 {
  bone1Name=extra_collar_02_r
  bone2Name=spine_05
  swing1LimitDegrees=45
  swing1Motion=Free
  swing2LimitDegrees=45
  swing2Motion=Free
  twistLimitDegrees=45
  twistMotion=Free
 }
 {
  bone1Name=extra_collar_02_l
  bone2Name=spine_05
  swing1LimitDegrees=45
  swing1Motion=Free
  swing2LimitDegrees=45
  swing2Motion=Free
  twistLimitDegrees=45
  twistMotion=Free
 }
]
